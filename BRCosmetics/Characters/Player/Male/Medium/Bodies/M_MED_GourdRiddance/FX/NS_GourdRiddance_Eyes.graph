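# Niagara system for the glowing cyan eyes of the GourdRiddance male player body cosmetic.
# One EyeGlow emitter bursts two particles, partitioned by index to follow separate skeletal mesh sockets or bones, with mirrored offsets and mesh scale.
# Both sprite and mesh renderers are present; a system Timeline drives alpha while a waveform pulses GlowColor.
# User.bisHeadHidden sets each particle's VisibilityTag.
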NiagaraSystem /BRCosmetics/Characters/Player/Male/Medium/Bodies/M_MED_GourdRiddance/FX/NS_GourdRiddance_Eyes
dependencies {
 usedDataInterfaces [&NiagaraDataInterfaceSkeletalMesh, &NiagaraDataInterfaceArrayFloat, &NiagaraDataInterfaceCurve]
 usedDynamicInputs [&Boolean_NotOperation]
 usedModules [
  &self:SystemSpawnScript.SystemScriptSource.SystemScriptGraph.NiagaraNodeAssignment_0.SetVariables_#1
  &SystemState
  &Timeline
 ]
}
emitters {EyeGlow {
  emitterSpawnScript [/SetVariables(
    Emitter.FlameSpeed:NiagaraFloat=&RandomRangeFloat(Minimum=1.9, Maximum=2.1)
    Emitter.UV Offset:NiagaraFloat=&RandomRangeFloat(Minimum=0.11, Maximum=0.99)
   ) {isSetParameters=true}]
  emitterUpdateScript [
   &EmitterState(
    Inactive Response=::NewEnumerator1
    Life Cycle Mode=::NewEnumerator1
    Loop Behavior=::NewEnumerator1
    Loop Duration Mode=::NewEnumerator1
    UseLoopDelay=false
   )
   &SpawnBurst_Instantaneous(Spawn Count=2, Spawn Probability=HlslExpression() {hlslExpression=1.0}, Use Spawn Probability=false, UseLoopCountLimit=false)
  ]
  particleSpawnScript [
   &PartitionParticles(ID to Partition:NiagaraInt32=&ReturnExecIndex())
   &InitializeParticle(
    Color=<rgba 0 1.7312 2 1>
    Mesh Scale=&Lerp_Vector(A=<1 1 1>, B=<1 1 -1>, Alpha=&MakeFloatFromInt(Int=$Particles.PartitionParticles.Partition))
    Mesh Scale Mode=::NewEnumerator4
    Sprite Size Mode=::NewEnumerator3
    Uniform Sprite Size=7
    UsePositionOffset=false
    Write Material Random=true
   )
   &InitialMeshOrientation(
    Mesh Axis To Orient=<0 1 0>
    Mesh Orientation Mode=::NewEnumerator2
    Orientation Vector=<0 1 0>
    Rotation=&MaskVectorBySpawnGroup(Masked Vector=<0 0.875 0>, Passthrough Vector=<0 0.125 0>, Spawn Group Mask=$Particles.PartitionParticles.Partition)
    Use Rotation Vector=true
   )
   /SetVariables() {isSetParameters=true}
  ]
  particleUpdateScript [
   &ParticleState(Kill Particles When Lifetime Has Elapsed=false, Let Infinitely Lived Particles Die When Emitter Deactivates=true)
   &CameraOffset(Camera Offset Amount=3)
   &SkeletalMeshLocation(
    Bone / Socket Index=$Particles.PartitionParticles.Partition
    MaskSpawnGroup=false
    Mesh Sampling Type=::NewEnumerator5
    Orientation Sampling=::NewEnumerator1
    Sampled Position Offset=&Lerp_Vector(
     A=<0.2 0.2 -0.5>
     B=<-0.2 0.2 -0.5>
     Alpha=&MakeFloatFromInt(Int=$Particles.PartitionParticles.Partition)
    )
    Skeletal Mesh=$System.Skeletal Mesh
    Socket Sampling Mode=::NewEnumerator5
   ) {module=SkeletalMeshLocation001}
   &ScaleColor(Scale Alpha=$System.Timeline.Float, ScaleA=true, ScaleRGB=true)
   /SetVariables(Particles.GlowColor=&MakeLinearColorFromVectorAndFloat(
     "Vector (RGB)"=&Multiply_VectorByFloat(
      Vector=&MakeVectorFromLinearColorRGB(Color=$Particles.Color)
      Float=&Waveform("Waveform [1]"=::NewEnumerator12, "[1] Frequency"=6, "[1] Amplitude Min/Max"=<0.5 1>)
     )
     "Float (Alpha)"=&Multiply_Float(A=0.3, B=$System.Timeline.Float)
    )) {isSetParameters=true}
   /SetVariables(Particles.VisibilityTag=&MakeCustomIntFromBool(Bool=$User.bisHeadHidden, True INT=1, False INT=0)) {isSetParameters=true}
  ]
  rendererClasses [&NiagaraSpriteRendererProperties, &NiagaraMeshRendererProperties]
  renderers [
   {rendererClass=&NiagaraSpriteRendererProperties}
   {rendererClass=&NiagaraMeshRendererProperties, rendererIndex=1}
  ]
 }}
summary {
 systemName=NS_GourdRiddance_Eyes
 emitters [{emitterName=EyeGlow, rendererClasses [&NiagaraSpriteRendererProperties, &NiagaraMeshRendererProperties]}]
 userVariables [
  User.bisADS {type=NiagaraBool}
  User.bisHeadHidden {type=NiagaraBool}
 ]
}
