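# Control Rig for the GourdRiddance male medium body cosmetic that corrects deformation of its pads, skirt, belt, feathers and scarf during movement.
# After a shared deform-rig function, a Variables function derives driver weights from relative bone rotations, thigh curves and movement states.
# Pads, Skirt, Belt and Scarf systems then run in sequence, blending weighted offsets and multi-parent constraints.
# Shoulder pads also use spherical pose readers, faded out while airborne.
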ControlRigBlueprint /BRCosmetics/Characters/Player/Male/Medium/Bodies/M_MED_GourdRiddance/Meshes/M_MED_GourdRiddance_Deform_CtrlRig
graphs [
 New Function_ContainedGraph {nodes [
   RigVMFunctionEntryNode()
   > RigVMFunction_Sequence = RigVMAggregateNode()
   RigVMFunction_MathBoolOr = &RigVMFunction_MathBoolOr(A=RigVMVariableNode() {variable=isCrouching}.Value, B=RigVMVariableNode() {variable=isRunning}.Value)
   DISPATCH_RigVMDispatch_If = RigVMDispatchNode(False=1, True=0.8, Condition=@RigVMFunction_MathBoolOr.Result) {template="DISPATCH_RigVMDispatch_If(in Condition,in True,in False,out Result)"}
   Multiply = &RigVMFunction_MathDoubleMul(
    B=RigVMDispatchNode(BoneName=root, Default=1, Name=deformRig_r_thigh_up_alpha) {template="DISPATCH_RigDispatch_GetAnimAttribute(in Name,in BoneName,in Default,out Value,out Found)"}.Value
    A=@DISPATCH_RigVMDispatch_If.Result
   )
   RigVMVariableNode(Value=@Multiply.Result, ExecuteContext=@RigVMFunction_Sequence.A) {variable=RunCrouch}
   DeltaRotation_Z = RigVMFunctionReferenceNode(Child=Head, Parent=neck_01, Source Maximum=-38, Source Minimum=-22, ExecuteContext=@RigVMFunction_Sequence.C)
   > RigVMVariableNode(Value=@DeltaRotation_Z.Delta) {variable=LookUp}
   > DeltaRotation_Z_1 = RigVMFunctionReferenceNode(Child=Head, Parent=neck_01, Source Maximum=0.5, Source Minimum=-10)
   > RigVMVariableNode(Value=@DeltaRotation_Z_1.Delta) {variable=LookDown}
   > DeltaRotation_Z_2 = RigVMFunctionReferenceNode(Child=spine_05, Parent=upperarm_l, Source Maximum=-140, Source Minimum=-100)
   > RigVMVariableNode(Value=@DeltaRotation_Z_2.Delta) {variable=aimDown}
   > DeltaRotation_Y = RigVMFunctionReferenceNode(Child=thigh_r, Parent=pelvis, Source Maximum=-75, Source Minimum=-7)
   > RigVMVariableNode(Value=@DeltaRotation_Y.Delta) {variable=thighSideRotateR}
   > DeltaRotation_Y_1 = RigVMFunctionReferenceNode(Child=thigh_l, Parent=pelvis, Source Maximum=-75, Source Minimum=-7)
   > RigVMVariableNode(Value=@DeltaRotation_Y_1.Delta) {variable=thighSideRotateL}
   RigVMVariableNode(
    Value=&RigVMFunction_MathDoubleRemap(SourceMaximum=1, SourceMinimum=0, TargetMaximum=0, TargetMinimum=1, bClamp=False, Value=RigVMVariableNode() {variable=floatingVar}.Value).Result
    ExecuteContext=@RigVMFunction_Sequence.B
   ) {variable=NotInAir}
   Remap = &RigVMFunction_MathDoubleRemap(SourceMaximum=1, SourceMinimum=0, TargetMaximum=0, TargetMinimum=1, bClamp=False, Value=RigVMVariableNode() {variable=moving2d}.Value)
   > RigVMVariableNode(Value=@Remap.Result) {variable=idleVar}
   RigVMFunctionReturnNode(ExecuteContext=@RigVMFunction_Sequence.D)
   DISPATCH_RigVMDispatch_Print = RigVMDispatchNode(Enabled=True, ExecuteContext=(), ScreenColor=(R=1,G=1,B=1,A=1), ScreenDuration=0.05, Value=0) {template="DISPATCH_RigVMDispatch_Print(in Prefix,in Value,in Enabled,in ScreenDuration,in ScreenColor)"}
  ]}
 New Function_ContainedGraph {nodes [
   RigVMFunctionEntryNode()
   > RigVMFunction_Sequence = RigVMAggregateNode()
   &RigUnit_ParentConstraint(
    AdvancedSettings=(InterpolationType=Average,RotationOrderForFilter=XZY)
    Child=(Type=Bone,Name="base_kneepad_l")
    Parents=((Item=(Type=Bone,Name="thigh_l"),Weight=0.5),(Item=(Type=Bone,Name="calf_l"),Weight=0.5))
    Weight=1
    ExecutePin=@RigVMFunction_Sequence.B
   )
   > &RigUnit_ParentConstraint(
    AdvancedSettings=(InterpolationType=Average,RotationOrderForFilter=XZY)
    Child=(Type=Bone,Name="base_kneepad_r")
    Parents=((Item=(Type=Bone,Name="thigh_r"),Weight=0.5),(Item=(Type=Bone,Name="calf_r"),Weight=0.5))
    Weight=1
   )
   > &RigUnit_ModifyTransforms(
    ItemToModify="((Item=(Type=Bone,Name=\"base_pant02_l\"),Transform=(Rotation=(X=0.232,Y=0.668,Z=-0.232,W=-0.668),Translation=(X=-12.0985,Y=-0.0155,Z=0.0139),Scale3D=(X=1,Y=1,Z=1))),(Item=(Type=Bone,Name=\"base_kneepad_l\"),Transform=(Rotation=(X=-0.018,Y=0.001,Z=-0.0534,W=-0 ...(truncated)"
    WeightMinimum=0
    Weight=RigVMDispatchNode(BoneName=root, Default=1, Name=deformRig_l_knee_alpha) {template="DISPATCH_RigDispatch_GetAnimAttribute(in Name,in BoneName,in Default,out Value,out Found)"}.Value
   )
   > &RigUnit_ModifyTransforms(
    ItemToModify="((Item=(Type=Bone,Name=\"base_pant02_r\"),Transform=(Rotation=(X=0,Y=0.7071,Z=0,W=-0.7071),Translation=(X=-12.0976,Y=0.001,Z=0),Scale3D=(X=1,Y=1,Z=1))),(Item=(Type=Bone,Name=\"base_kneepad_r\"),Transform=(Rotation=(X=0.061,Y=0.008,Z=0.1303,W=0.9896),Translatio ...(truncated)"
    WeightMinimum=0
    Weight=RigVMDispatchNode(BoneName=root, Default=1, Name=deformRig_r_knee_alpha) {template="DISPATCH_RigDispatch_GetAnimAttribute(in Name,in BoneName,in Default,out Value,out Found)"}.Value
   )
   RigVMFunctionReturnNode(ExecuteContext=@RigVMFunction_Sequence.C)
   n1 = RigVMVariableNode() {variable=floatingVar}
   &RigUnit_ModifyTransforms(
    ItemToModify="((Item=(Type=Bone,Name=\"base_shoulderPad_l\"),Transform=(Rotation=(X=0.2005,Y=0.0887,Z=-0.1048,W=0.97),Translation=(X=6.7775,Y=0.5217,Z=-0.8278),Scale3D=(X=0.9,Y=0.9,Z=0.9))),(Item=(Type=Bone,Name=\"base_shoulderPDM_l\"),Transform=(Rotation=(X=0,Y=-0.3011,Z=0 ...(truncated)"
    WeightMinimum=0
    Weight=@n1.Value
    ExecutePin=@RigVMFunction_Sequence.A
   )
   > &RigUnit_ModifyTransforms(
    ItemToModify=((Item=(Type=Bone,Name="base_shoulderPad_r"),Transform=(Rotation=(X=0.4307,Y=0,Z=0,W=0.9025),Translation=(X=-7.6525,Y=0,Z=0.0001),Scale3D=(X=0.9,Y=0.9,Z=0.9))))
    WeightMinimum=0
    Weight=@n1.Value
   )
   > n2 = &RigUnit_SphericalPoseReader(
    ActiveRegionScaleFactors=(PositiveWidth=0.5,NegativeWidth=1,PositiveHeight=1,NegativeHeight=1)
    ActiveRegionSize=0.1
    Debug=(bDrawDebug=false,bDraw2D=false,bDrawLocalAxes=false,DebugScale=25,DebugSegments=20,DebugThickness=0.25)
    DriverAxis=(X=1,Y=0,Z=0)
    DriverItem=(Type=Bone,Name="clavicle_l")
    FalloffRegionScaleFactors=(PositiveWidth=1.5,NegativeWidth=1,PositiveHeight=1,NegativeHeight=1)
    FalloffSize=0.2
    FlipHeightScaling=False
    FlipWidthScaling=False
    OptionalParentItem=(Type=Bone,Name="None")
    RotationOffset=(X=0,Y=-40,Z=0)
   )
   > &RigUnit_ModifyTransforms(
    ItemToModify="((Item=(Type=Bone,Name=\"base_shoulderPad_l\"),Transform=(Rotation=(X=-0.1949,Y=-0.1003,Z=0.1614,W=-0.9622),Translation=(X=9.2347,Y=-0.1974,Z=-1.5059),Scale3D=(X=0.9,Y=0.9,Z=0.9))),(Item=(Type=Bone,Name=\"base_shoulderPDM_l\"),Transform=(Rotation=(X=0,Y=-0.301 ...(truncated)"
    WeightMinimum=0
    Weight=&RigVMFunction_MathDoubleMul(B=@n2.OutputParam, A=RigVMVariableNode() {variable=NotInAir}.Value).Result
   )
   > &RigUnit_ModifyTransforms(
    ItemToModify=((Item=(Type=Bone,Name="base_shoulderPad_r"),Transform=(Rotation=(X=0.3423,Y=0,Z=0,W=0.9396),Translation=(X=-9.6894,Y=0,Z=0),Scale3D=(X=0.9,Y=0.9,Z=0.9))))
    WeightMinimum=0
    Weight=&RigVMFunction_MathDoubleMul(
     B=RigVMDispatchNode(BoneName=root, Default=1, Name=deformRig_r_upperarm_up_alpha) {template="DISPATCH_RigDispatch_GetAnimAttribute(in Name,in BoneName,in Default,out Value,out Found)"}.Value
     A=RigVMVariableNode() {variable=NotInAir}.Value
    ).Result
   )
   > &RigUnit_ModifyTransforms(
    ItemToModify=((Item=(Type=Bone,Name="base_shoulderPad_r"),Transform=(Rotation=(X=-0.3219,Y=0,Z=0,W=-0.9468),Translation=(X=0,Y=0,Z=0),Scale3D=(X=1,Y=1,Z=1))))
    WeightMinimum=0
    Weight=&RigVMFunction_MathDoubleMul(
     B=RigVMDispatchNode(BoneName=root, Default=1, Name=deformRig_r_upperarm_fr_alpha) {template="DISPATCH_RigDispatch_GetAnimAttribute(in Name,in BoneName,in Default,out Value,out Found)"}.Value
     A=RigVMVariableNode() {variable=NotInAir}.Value
    ).Result
   )
   > n3 = &RigUnit_SphericalPoseReader(
    ActiveRegionScaleFactors=(PositiveWidth=1,NegativeWidth=1,PositiveHeight=1,NegativeHeight=0.8)
    ActiveRegionSize=0.1
    Debug=(bDrawDebug=false,bDraw2D=False,bDrawLocalAxes=False,DebugScale=25,DebugSegments=20,DebugThickness=0.25)
    DriverAxis=(X=1,Y=0,Z=0)
    DriverItem=(Type=Bone,Name="upperarm_r")
    FalloffRegionScaleFactors=(PositiveWidth=1,NegativeWidth=1,PositiveHeight=1.7,NegativeHeight=0.5)
    FalloffSize=0.2
    FlipHeightScaling=False
    FlipWidthScaling=False
    OptionalParentItem=(Type=Bone,Name="None")
    RotationOffset=(X=51.9395,Y=8.7023,Z=22.4256)
   )
   > &RigUnit_ModifyTransforms(
    ItemToModify=((Item=(Type=Bone,Name="base_shoulderPad_l"),Transform=(Rotation=(X=0,Y=0,Z=0.0719,W=0.9974),Translation=(X=-8.0691,Y=-1.4457,Z=1.5019),Scale3D=(X=1,Y=1,Z=1))))
    WeightMinimum=0
    Weight=&RigVMFunction_MathDoubleMul(A=@n3.OutputParam, B=RigVMVariableNode() {variable=NotInAir}.Value).Result
   )
  ]}
 New Function_ContainedGraph {nodes [
   n1 = RigVMFunctionEntryNode()
   RigVMFunction_Sequence = &RigVMFunction_Sequence(ExecuteContext=@n1.ExecuteContext)
   SequenceAggregate_3_1_1 = RigVMAggregateNode(ExecuteContext=@RigVMFunction_Sequence.A)
   &RigUnit_ParentConstraint(
    AdvancedSettings=(InterpolationType=Average,RotationOrderForFilter=XZY)
    Child=(Type=Bone,Name="base_skirt_G")
    Filter=(TranslationFilter=(bX=True,bY=True,bZ=True),RotationFilter=(bX=True,bY=True,bZ=True),ScaleFilter=(bX=false,bY=false,bZ=false))
    Parents=((Item=(Type=Bone,Name="deform_groin_r"),Weight=1))
    Weight=1
    ExecutePin=@SequenceAggregate_3_1_1.B
   )
   > &RigUnit_ParentConstraint(
    AdvancedSettings=(RotationOrderForFilter=XZY)
    Child=(Type=Bone,Name="dyn_skirt_F_01")
    Filter=(TranslationFilter=(bX=False,bY=False,bZ=False),RotationFilter=(bX=True,bY=True,bZ=True),ScaleFilter=(bX=True,bY=True,bZ=True))
    Parents=((Item=(Type=Bone,Name="dyn_skirt_G_01"),Weight=0.5),(Item=(Type=Bone,Name="base_skirt_E"),Weight=0.5))
    Weight=1
   )
   > &RigUnit_ParentConstraint(
    AdvancedSettings=(RotationOrderForFilter=XZY)
    Child=(Type=Bone,Name="base_skirt_E")
    Filter=(TranslationFilter=(bX=False,bY=False,bZ=False),RotationFilter=(bX=True,bY=True,bZ=True),ScaleFilter=(bX=True,bY=True,bZ=True))
    Parents=((Item=(Type=Bone,Name="dyn_skirt_F_01"),Weight=0.5),(Item=(Type=Bone,Name="base_skirt_D"),Weight=0.5))
    Weight=1
   )
   n2 = RigVMDispatchNode(BoneName=root, Default=1, Name=deformRig_l_thigh_up_alpha) {template="DISPATCH_RigDispatch_GetAnimAttribute(in Name,in BoneName,in Default,out Value,out Found)"}
   n3 = &RigVMFunction_MathDoubleRemap(SourceMaximum=1, SourceMinimum=0, TargetMaximum=0.6, TargetMinimum=0, bClamp=False, Value=RigVMVariableNode() {variable=moving2d}.Value)
   &RigUnit_ModifyTransforms(
    ItemToModify="((Item=(Type=Bone,Name=\"dyn_skirt_A_01\"),Transform=(Rotation=(X=0.0516,Y=0.3008,Z=-0.0837,W=0.9486),Translation=(X=6.1906,Y=-1.2125,Z=-3.6304),Scale3D=(X=1,Y=1,Z=1))),(Item=(Type=Bone,Name=\"dyn_skirt_A_02\"),Transform=(Rotation=(X=-0.0148,Y=-0.167,Z=0.024,W ...(truncated)"
    WeightMinimum=0
    Weight=&RigVMFunction_MathDoubleRemap(
     SourceMaximum=1.2
     SourceMinimum=0
     TargetMaximum=1
     TargetMinimum=0
     bClamp=False
     Value=&RigVMFunction_MathDoubleAdd(B=@n2.Value, A=@n3.Result).Result
    ).Result
    ExecutePin=@SequenceAggregate_3_1_1.C
   )
   n4 = RigVMDispatchNode(BoneName=root, Default=1, Name=deformRig_l_thigh_up_alpha) {template="DISPATCH_RigDispatch_GetAnimAttribute(in Name,in BoneName,in Default,out Value,out Found)"}
   > &RigUnit_ModifyTransforms(
    ItemToModify=((Item=(Type=Bone,Name="base_skirt_A"),Transform=(Rotation=(X=0,Y=0.0305,Z=0,W=-0.9995),Translation=(X=4.7435,Y=0,Z=-4.7521),Scale3D=(X=1,Y=1,Z=1))))
    WeightMinimum=0
    Weight=&RigVMFunction_MathDoubleMul(
     B=&RigVMFunction_MathDoubleRemap(SourceMaximum=1, SourceMinimum=0, TargetMaximum=0, TargetMinimum=1, bClamp=False, Value=@n4.Value).Result
     A=RigVMVariableNode() {variable=idleVar}.Value
    ).Result
   )
   n5 = RigVMDispatchNode(BoneName=root, Default=1, Name=deformRig_r_thigh_up_alpha) {template="DISPATCH_RigDispatch_GetAnimAttribute(in Name,in BoneName,in Default,out Value,out Found)"}
   n6 = RigVMVariableNode() {variable=moving2d}
   &RigUnit_ModifyTransforms(
    ItemToModify="((Item=(Type=Bone,Name=\"dyn_skirt_G_01\"),Transform=(Rotation=(X=-0.0631,Y=-0.4006,Z=-0.0895,W=0.9097),Translation=(X=0.5481,Y=2.1102,Z=0),Scale3D=(X=1,Y=1,Z=1))),(Item=(Type=Bone,Name=\"dyn_skirt_G_02\"),Transform=(Rotation=(X=0.0006,Y=0.342,Z=-0.0128,W=0.93 ...(truncated)"
    WeightMinimum=0
    Weight=&RigVMFunction_MathDoubleRemap(
     SourceMaximum=1.2
     SourceMinimum=0
     TargetMaximum=1
     TargetMinimum=0
     bClamp=False
     Value=&RigVMFunction_MathDoubleAdd(B=@n5.Value, A=@n6.Value).Result
    ).Result
    ExecutePin=@SequenceAggregate_3_1_1.D
   )
   Multiply = &RigVMFunction_MathDoubleMul(
    B=RigVMDispatchNode(BoneName=root, Default=1, Name=deformRig_r_thigh_up_alpha) {template="DISPATCH_RigDispatch_GetAnimAttribute(in Name,in BoneName,in Default,out Value,out Found)"}.Value
    A=RigVMVariableNode() {variable=IsRunFloat}.Value
   )
   > &RigUnit_ModifyTransforms(
    ItemToModify="((Item=(Type=Bone,Name=\"dyn_skirt_G_01\"),Transform=(Rotation=(X=-0.002,Y=0.0175,Z=-0.1308,W=-0.9913),Translation=(X=0,Y=0,Z=0),Scale3D=(X=1,Y=1,Z=1))),(Item=(Type=Bone,Name=\"dyn_skirt_G_02\"),Transform=(Rotation=(X=0.0818,Y=-0.0262,Z=-0.1044,W=0.9908),Trans ...(truncated)"
    WeightMinimum=0
    Weight=@Multiply.Result
   )
   &RigUnit_ParentConstraint(
    AdvancedSettings=(InterpolationType=Average,RotationOrderForFilter=XZY)
    Child=(Type=Bone,Name="base_skirt_A")
    Filter=(TranslationFilter=(bX=true,bY=true,bZ=true),RotationFilter=(bX=true,bY=true,bZ=true),ScaleFilter=(bX=false,bY=false,bZ=false))
    Parents=((Item=(Type=Bone,Name="deform_groin_l"),Weight=1))
    Weight=1
    ExecutePin=@SequenceAggregate_3_1_1.A
   )
   > &RigUnit_ParentConstraint(
    AdvancedSettings=(RotationOrderForFilter=XZY)
    Child=(Type=Bone,Name="base_skirt_B")
    Filter=(TranslationFilter=(bX=False,bY=False,bZ=False),RotationFilter=(bX=True,bY=True,bZ=True),ScaleFilter=(bX=true,bY=true,bZ=true))
    Parents=((Item=(Type=Bone,Name="dyn_skirt_A_01"),Weight=0.5),(Item=(Type=Bone,Name="base_skirt_C"),Weight=0.5))
    Weight=1
   )
   > &RigUnit_ParentConstraint(
    AdvancedSettings=(RotationOrderForFilter=XZY)
    Child=(Type=Bone,Name="base_skirt_C")
    Filter=(TranslationFilter=(bX=False,bY=False,bZ=False),RotationFilter=(bX=True,bY=True,bZ=True),ScaleFilter=(bX=true,bY=true,bZ=true))
    Parents=((Item=(Type=Bone,Name="base_skirt_B"),Weight=0.5),(Item=(Type=Bone,Name="base_skirt_D"),Weight=0.5))
    Weight=1
   )
   n7 = RigVMVariableNode() {variable=moving2d}
   &RigUnit_ModifyTransforms(
    ItemToModify="((Item=(Type=Bone,Name=\"base_skirt_G\"),Transform=(Rotation=(X=0,Y=-0.1305,Z=0,W=0.9914),Translation=(X=0,Y=0,Z=0),Scale3D=(X=1,Y=1,Z=1))),(Item=(Type=Bone,Name=\"base_skirt_F\"),Transform=(Rotation=(X=-0.1039,Y=-0.0985,Z=-0.3571,W=0.923),Translation=(X=0,Y=0 ...(truncated)"
    WeightMinimum=0
    Weight=&RigVMFunction_MathDoubleMul(
     A=&RigVMFunction_MathDoubleRemap(SourceMaximum=1, SourceMinimum=0, TargetMaximum=0.1, TargetMinimum=1, bClamp=False, Value=@n7.Value).Result
     B=RigVMVariableNode() {variable=thighSideRotateR}.Value
    ).Result
    ExecutePin=@SequenceAggregate_3_1_1.E
   )
   n8 = RigVMVariableNode() {variable=moving2d}
   > &RigUnit_ModifyTransforms(
    ItemToModify="((Item=(Type=Bone,Name=\"dyn_skirt_B_01\"),Transform=(Rotation=(X=0,Y=0,Z=-0.5628,W=0.8266),Translation=(X=0,Y=0,Z=0),Scale3D=(X=1,Y=1,Z=1))),(Item=(Type=Bone,Name=\"base_skirt_C\"),Transform=(Rotation=(X=-0.0015,Y=-0.0375,Z=-0.0849,W=0.9957),Translation=(X=0, ...(truncated)"
    WeightMinimum=0
    Weight=&RigVMFunction_MathDoubleMul(
     A=&RigVMFunction_MathDoubleRemap(SourceMaximum=1, SourceMinimum=0, TargetMaximum=0.1, TargetMinimum=1, bClamp=False, Value=@n8.Value).Result
     B=RigVMVariableNode() {variable=thighSideRotateL}.Value
    ).Result
   )
   RigVMFunctionReturnNode(ExecuteContext=@RigVMFunction_Sequence.B)
   DISPATCH_RigVMDispatch_Print = RigVMDispatchNode(Enabled=True, ExecuteContext=(), ScreenColor=(R=1,G=1,B=1,A=1), ScreenDuration=0.05, Value=RigVMVariableNode() {variable=isSliding}.Value) {template="DISPATCH_RigVMDispatch_Print(in Prefix,in Value,in Enabled,in ScreenDuration,in ScreenColor)"}
   Remap = &RigVMFunction_MathDoubleRemap(
    SourceMaximum=30
    SourceMinimum=50
    TargetMaximum=1
    TargetMinimum=0
    bClamp=False
    Value=&RigUnit_GetRelativeTransformForItem(Child=(Type=Bone,Name="spine_01"), Parent=(Type=Bone,Name="root"), bChildInitial=False, bParentInitial=False).RelativeTransform.Translation.Z
   )
   RigVMVariableNode(ExecuteContext=(), Value=@Remap.Result) {variable=isSliding}
  ]}
 New Function_ContainedGraph {nodes [
   n1 = RigVMFunctionEntryNode()
   RigVMFunction_Sequence = &RigVMFunction_Sequence(ExecuteContext=@n1.ExecuteContext)
   &RigUnit_ModifyTransforms(
    ItemToModify=((Item=(Type=Bone,Name="base_inbelt_l"),Transform=(Rotation=(X=0,Y=0,Z=-0.2044,W=0.9789),Translation=(X=-11.6251,Y=-0.0193,Z=2.9108),Scale3D=(X=0.6203,Y=1,Z=1))))
    WeightMinimum=0
    Weight=RigVMDispatchNode(BoneName=root, Default=1, Name=deformRig_l_thigh_up_alpha) {template="DISPATCH_RigDispatch_GetAnimAttribute(in Name,in BoneName,in Default,out Value,out Found)"}.Value
    ExecutePin=@RigVMFunction_Sequence.A
   )
   > &RigUnit_ModifyTransforms(
    ItemToModify=((Item=(Type=Bone,Name="base_inbelt_r"),Transform=(Rotation=(X=0,Y=0,Z=-0.1703,W=-0.9854),Translation=(X=-11.7437,Y=-0.5477,Z=3.7254),Scale3D=(X=0.4732,Y=1,Z=1))))
    WeightMinimum=0
    Weight=RigVMDispatchNode(BoneName=root, Default=1, Name=deformRig_r_thigh_up_alpha) {template="DISPATCH_RigDispatch_GetAnimAttribute(in Name,in BoneName,in Default,out Value,out Found)"}.Value
   )
   > &RigUnit_ModifyTransforms(
    ItemToModify="((Item=(Type=Bone,Name=\"dyn_outbelt_side_l\"),Transform=(Rotation=(X=0.077,Y=-0.0517,Z=0.1024,W=0.9904),Translation=(X=-3.8222,Y=4.6229,Z=0.5666),Scale3D=(X=1,Y=1,Z=1))),(Item=(Type=Bone,Name=\"dyn_outbelt_bck_l\"),Transform=(Rotation=(X=-0.0306,Y=-0.1811,Z=0 ...(truncated)"
    WeightMinimum=0
    Weight=RigVMDispatchNode(BoneName=root, Default=1, Name=deformRig_l_thigh_up_alpha) {template="DISPATCH_RigDispatch_GetAnimAttribute(in Name,in BoneName,in Default,out Value,out Found)"}.Value
    ItemToModify.3.Transform.Translation=RigVMDispatchNode(False=(X=-12.5785,Y=8.0927,Z=13.8019), True=(X=-7.7931,Y=3.1917,Z=16.2288), Condition=RigVMVariableNode() {variable=isCrouching}.Value) {template="DISPATCH_RigVMDispatch_If(in Condition,in True,in False,out Result)"}.Result
   )
   > &RigUnit_ModifyTransforms(
    ItemToModify=((Item=(Type=Bone,Name="dyn_outbelt_bck_l"),Transform=(Rotation=(X=0.2588,Y=0,Z=0,W=-0.9659),Translation=(X=0,Y=-3,Z=3),Scale3D=(X=1,Y=1.7227,Z=1))))
    WeightMinimum=0
    Weight=RigVMDispatchNode(BoneName=root, Default=1, Name=deformRig_r_thigh_up_alpha) {template="DISPATCH_RigDispatch_GetAnimAttribute(in Name,in BoneName,in Default,out Value,out Found)"}.Value
   )
   > &RigUnit_ModifyTransforms(
    ItemToModify=((Item=(Type=Bone,Name="base_outbelt_r"),Transform=(Rotation=(X=0.0725,Y=0.1977,Z=0.3366,W=-0.9178),Translation=(X=-11.3824,Y=-0.3933,Z=2.6758),Scale3D=(X=0.7655,Y=1.5498,Z=1))))
    WeightMinimum=0
    Weight=RigVMVariableNode() {variable=RunCrouch}.Value
   )
   > &RigUnit_ModifyTransforms(
    ItemToModify="((Item=(Type=Bone,Name=\"base_outbelt_r\"),Transform=(Rotation=(X=0,Y=0,Z=0.331,W=0.9436),Translation=(X=-12,Y=0.447,Z=1.2),Scale3D=(X=1,Y=1,Z=1))),(Item=(Type=Bone,Name=\"dyn_outbelt_bck_l\"),Transform=(Rotation=(X=0,Y=0,Z=0,W=1),Translation=(X=-15,Y=0,Z=0.85 ...(truncated)"
    WeightMinimum=0
    Weight=RigVMVariableNode() {variable=isSliding}.Value
   )
   > &RigUnit_ParentConstraint(
    AdvancedSettings=(RotationOrderForFilter=XZY)
    Child=(Type=Bone,Name="base_feather_d01_r")
    Parents="((Item=(Type=Bone,Name=\"spine_03\"),Weight=0.0006),(Item=(Type=Bone,Name=\"spine_01\"),Weight=0.0011),(Item=(Type=Bone,Name=\"spine_02\"),Weight=0.0013),(Item=(Type=Bone,Name=\"deform_glute_r\"),Weight=0.0197),(Item=(Type=Bone,Name=\"dyn_outbelt_r\"),Weight=0.4),(I ...(truncated)"
    Weight=1
   )
   > &RigUnit_ParentConstraint(
    AdvancedSettings=(RotationOrderForFilter=XZY)
    Child=(Type=Bone,Name="base_feather_c01_r")
    Parents=((Item=(Type=Bone,Name="spine_03"),Weight=0.0044),(Item=(Type=Bone,Name="spine_01"),Weight=0.0078),(Item=(Type=Bone,Name="spine_02"),Weight=0.0095),(Item=(Type=Bone,Name="dyn_outbelt_r"),Weight=0.4),(Item=(Type=Bone,Name="pelvis"),Weight=0.5783))
    Weight=1
   )
   > &RigUnit_ParentConstraint(
    AdvancedSettings=(RotationOrderForFilter=XZY)
    Child=(Type=Bone,Name="base_feather_b01_r")
    Parents="((Item=(Type=Bone,Name=\"thigh_twist_01_l\"),Weight=0.0005),(Item=(Type=Bone,Name=\"spine_03\"),Weight=0.0015),(Item=(Type=Bone,Name=\"spine_01\"),Weight=0.0026),(Item=(Type=Bone,Name=\"spine_02\"),Weight=0.0044),(Item=(Type=Bone,Name=\"dyn_outbelt_r\"),Weight=0.4), ...(truncated)"
    Weight=1
   )
   > &RigUnit_ParentConstraint(
    AdvancedSettings=(RotationOrderForFilter=XZY)
    Child=(Type=Bone,Name="base_feather_a01_r")
    Parents="((Item=(Type=Bone,Name=\"spine_03\"),Weight=0.0009),(Item=(Type=Bone,Name=\"spine_01\"),Weight=0.0015),(Item=(Type=Bone,Name=\"spine_02\"),Weight=0.0036),(Item=(Type=Bone,Name=\"deform_groin_l\"),Weight=0.0067),(Item=(Type=Bone,Name=\"thigh_twist_01_l\"),Weight=0.01 ...(truncated)"
    Weight=1
   )
   > &RigUnit_ParentConstraint(
    AdvancedSettings=(RotationOrderForFilter=XZY)
    Child=(Type=Bone,Name="base_feather_e01_r")
    Parents=((Item=(Type=Bone,Name="spine_01"),Weight=0.0022),(Item=(Type=Bone,Name="spine_02"),Weight=0.0026),(Item=(Type=Bone,Name="deform_glute_r"),Weight=0.0513),(Item=(Type=Bone,Name="dyn_outbelt_r"),Weight=0.4),(Item=(Type=Bone,Name="pelvis"),Weight=0.5438))
    Weight=1
   )
   > &RigUnit_ParentConstraint(
    AdvancedSettings=(RotationOrderForFilter=XZY)
    Child=(Type=Bone,Name="base_outbelt_bck_l")
    Parents="((Item=(Type=Bone,Name=\"dyn_skirt_D_03\"),Weight=0.0017),(Item=(Type=Bone,Name=\"deform_glute_l\"),Weight=0.0017),(Item=(Type=Bone,Name=\"dyn_skirt_C_03\"),Weight=0.0018),(Item=(Type=Bone,Name=\"pelvis\"),Weight=0.0455),(Item=(Type=Bone,Name=\"dyn_skirt_D_02\"),Wei ...(truncated)"
    Weight=1
   )
   > &RigUnit_ParentConstraint(
    AdvancedSettings=(RotationOrderForFilter=XZY)
    Child=(Type=Bone,Name="base_beltBag_l")
    Parents="((Item=(Type=Bone,Name=\"deform_glute_l\"),Weight=0.0791),(Item=(Type=Bone,Name=\"spine_03\"),Weight=0.0877),(Item=(Type=Bone,Name=\"spine_02\"),Weight=0.1038),(Item=(Type=Bone,Name=\"spine_01\"),Weight=0.1116),(Item=(Type=Bone,Name=\"deform_groin_l\"),Weight=0.1227 ...(truncated)"
    Weight=1
   )
   > &RigUnit_ParentConstraint(
    AdvancedSettings=(RotationOrderForFilter=XZY)
    Child=(Type=Bone,Name="base_outbelt_side_l")
    Parents="((Item=(Type=Bone,Name=\"spine_02\"),Weight=0.0006),(Item=(Type=Bone,Name=\"deform_groin_l\"),Weight=0.0007),(Item=(Type=Bone,Name=\"thigh_twist_01_l\"),Weight=0.0021),(Item=(Type=Bone,Name=\"dyn_skirt_C_02\"),Weight=0.0039),(Item=(Type=Bone,Name=\"dyn_skirt_C_01\") ...(truncated)"
    Weight=1
   )
   > &RigUnit_ParentConstraint(
    AdvancedSettings=(RotationOrderForFilter=XZY)
    Child=(Type=Bone,Name="base_outbelt_fr_l")
    Parents="((Item=(Type=Bone,Name=\"thigh_twist_01_r\"),Weight=0.0098),(Item=(Type=Bone,Name=\"spine_01\"),Weight=0.0114),(Item=(Type=Bone,Name=\"spine_02\"),Weight=0.0143),(Item=(Type=Bone,Name=\"thigh_twist_01_l\"),Weight=0.0255),(Item=(Type=Bone,Name=\"dyn_skirt_A_03\"),Wei ...(truncated)"
    Weight=1
   )
   > &RigUnit_RotationConstraintLocalSpaceOffset(
    AdvancedSettings=(InterpolationType=Average,RotationOrderForFilter=XZY)
    Child=(Type=Bone,Name="base_feather_e01_r")
    Filter=(bX=true,bY=true,bZ=true)
    Parents=((Item=(Type=Bone,Name="dyn_skirt_E_01"),Weight=0.6),(Item=(Type=Bone,Name="pelvis"),Weight=0.4))
    Weight=1
   )
   > &RigUnit_RotationConstraintLocalSpaceOffset(
    AdvancedSettings=(InterpolationType=Average,RotationOrderForFilter=XZY)
    Child=(Type=Bone,Name="base_feather_b01_r")
    Filter=(bX=true,bY=true,bZ=false)
    Parents=((Item=(Type=Bone,Name="dyn_skirt_F_01"),Weight=1),(Item=(Type=Bone,Name="pelvis"),Weight=0))
    Weight=1
   )
   > &RigUnit_RotationConstraintLocalSpaceOffset(
    AdvancedSettings=(InterpolationType=Average,RotationOrderForFilter=XZY)
    Child=(Type=Bone,Name="base_feather_c01_r")
    Filter=(bX=false,bY=true,bZ=false)
    Parents=((Item=(Type=Bone,Name="dyn_skirt_F_01"),Weight=1),(Item=(Type=Bone,Name="pelvis"),Weight=0))
    Weight=1
   )
   > &RigUnit_RotationConstraintLocalSpaceOffset(
    AdvancedSettings=(InterpolationType=Average,RotationOrderForFilter=XZY)
    Child=(Type=Bone,Name="base_feather_d01_r")
    Filter=(bX=false,bY=true,bZ=false)
    Parents=((Item=(Type=Bone,Name="dyn_skirt_F_01"),Weight=1),(Item=(Type=Bone,Name="pelvis"),Weight=0))
    Weight=1
   )
   > &RigUnit_ModifyTransforms(
    ItemToModify="((Item=(Type=Bone,Name=\"base_feather_a01_r\"),Transform=(Rotation=(X=-0.0357,Y=0.6788,Z=-0.1729,W=-0.7128),Translation=(X=0.5658,Y=0.1138,Z=-0.0756),Scale3D=(X=1,Y=1,Z=1))),(Item=(Type=Bone,Name=\"base_feather_b01_r\"),Transform=(Rotation=(X=0.1256,Y=0.4829,Z ...(truncated)"
    WeightMinimum=0
    Weight=RigVMVariableNode() {variable=RunCrouch}.Value
   )
   > &RigUnit_ModifyTransforms(
    ItemToModify="((Item=(Type=Bone,Name=\"base_feather_b01_r\"),Transform=(Rotation=(X=0,Y=0.1201,Z=0,W=-0.9928),Translation=(X=0,Y=0,Z=0),Scale3D=(X=1,Y=1,Z=1))),(Item=(Type=Bone,Name=\"base_feather_c01_r\"),Transform=(Rotation=(X=0,Y=0.0774,Z=0,W=-0.997),Translation=(X=0,Y=0 ...(truncated)"
    WeightMinimum=0
    Weight=RigVMVariableNode() {variable=floatingVar}.Value
   )
   > &RigUnit_ModifyTransforms(
    ItemToModify="((Item=(Type=Bone,Name=\"base_feather_a01_r\"),Transform=(Rotation=(X=0,Y=0.1514,Z=0,W=-0.9885),Translation=(X=0,Y=0,Z=0),Scale3D=(X=1,Y=1,Z=1))),(Item=(Type=Bone,Name=\"base_feather_c01_r\"),Transform=(Rotation=(X=-0.2781,Y=-0.0655,Z=0.019,W=0.9581),Translati ...(truncated)"
    WeightMinimum=0
    Weight=RigVMVariableNode() {variable=thighSideRotateR}.Value
   )
   RigVMFunctionReturnNode(ExecuteContext=@RigVMFunction_Sequence.B)
  ]}
 New Function_ContainedGraph {nodes [
   n1 = RigVMFunctionEntryNode()
   RigVMFunction_MathQuaternionToEuler = &RigVMFunction_MathQuaternionToEuler(
    RotationOrder=ZYX
    Value=&RigUnit_GetRelativeTransformForItem(
     Child=(Type=Bone,Name="hand_l")
     Parent=(Type=Bone,Name="lowerarm_l")
     bChildInitial=False
     bParentInitial=False
     Child.Name=@n1.Child
     Parent.Name=@n1.Parent
    ).RelativeTransform.Rotation
   )
   Remap = &RigVMFunction_MathDoubleRemap(
    TargetMaximum=1
    TargetMinimum=0
    Value=@RigVMFunction_MathQuaternionToEuler.Result.Z
    SourceMaximum=@n1.Source Maximum
    SourceMinimum=@n1.Source Minimum
   )
   > RigVMFunctionReturnNode(Delta=@Remap.Result)
  ]}
 New Function_ContainedGraph {nodes [
   n1 = RigVMFunctionEntryNode()
   RigVMFunction_MathQuaternionToEuler = &RigVMFunction_MathQuaternionToEuler(
    RotationOrder=ZYX
    Value=&RigUnit_GetRelativeTransformForItem(
     Child=(Type=Bone,Name="hand_l")
     Parent=(Type=Bone,Name="lowerarm_l")
     bChildInitial=False
     bParentInitial=False
     Child.Name=@n1.Child
     Parent.Name=@n1.Parent
    ).RelativeTransform.Rotation
   )
   Remap = &RigVMFunction_MathDoubleRemap(
    TargetMaximum=1
    TargetMinimum=0
    Value=@RigVMFunction_MathQuaternionToEuler.Result.X
    SourceMaximum=@n1.Source Maximum
    SourceMinimum=@n1.Source Minimum
   )
   > RigVMFunctionReturnNode(Delta=@Remap.Result)
  ]}
 New Function_ContainedGraph {nodes [
   n1 = RigVMFunctionEntryNode()
   RigVMFunction_MathQuaternionToEuler = &RigVMFunction_MathQuaternionToEuler(
    RotationOrder=XYZ
    Value=&RigUnit_GetRelativeTransformForItem(
     Child=(Type=Bone,Name="hand_l")
     Parent=(Type=Bone,Name="lowerarm_l")
     bChildInitial=False
     bParentInitial=False
     Child.Name=@n1.Child
     Parent.Name=@n1.Parent
    ).RelativeTransform.Rotation
   )
   Remap = &RigVMFunction_MathDoubleRemap(
    TargetMaximum=1
    TargetMinimum=0
    Value=@RigVMFunction_MathQuaternionToEuler.Result.Y
    SourceMaximum=@n1.Source Maximum
    SourceMinimum=@n1.Source Minimum
   )
   > RigVMFunctionReturnNode(Delta=@Remap.Result)
  ]}
 New Function_ContainedGraph {nodes [
   n1 = RigVMFunctionEntryNode()
   RigVMFunction_ControlFlowBranch = &RigVMFunction_ControlFlowBranch(ExecuteContext=@n1.Execute Content, Condition=@n1.debug)
   &RigUnit_SetTransform(
    Item=(Type=Bone,Name="upperarm_l")
    Weight=1
    bInitial=False
    Value=&RigUnit_GetTransform(Item=(Type=Control,Name="upperarm_l_ctrl"), bInitial=False).Transform
    ExecutePin=@RigVMFunction_ControlFlowBranch.True
   )
   RigUnit_GetTransform = &RigUnit_GetTransform(Item=(Type=Control,Name="thigh_r_ctrl"), bInitial=False)
   > &RigUnit_SetTransform(Item=(Type=Bone,Name="thigh_r"), Weight=1, bInitial=False, Value=@RigUnit_GetTransform.Transform)
   > &RigUnit_SetTransform(
    Item=(Type=Bone,Name="thigh_l")
    Weight=1
    bInitial=False
    Value=&RigUnit_GetTransform(Item=(Type=Control,Name="thigh_l_ctrl"), bInitial=False).Transform
   )
   > &RigUnit_SetTransform(
    Item=(Type=Bone,Name="dyn_skirt_D_01")
    Weight=1
    bInitial=False
    Value=&RigUnit_GetTransform(Item=(Type=Control,Name="dyn_skirt_D_01_ctrl"), bInitial=False).Transform
   )
   RigVMFunctionReturnNode(Execute Content=@RigVMFunction_ControlFlowBranch.Completed)
  ]}
 New Function_ContainedGraph {nodes [
   RigVMFunctionEntryNode()
   > &RigUnit_ParentConstraint(
    AdvancedSettings=(RotationOrderForFilter=XZY)
    Child=(Type=Bone,Name="base_tipScarf_m")
    Parents="((Item=(Type=Bone,Name=\"clavicle_l\"),Weight=0.0114),(Item=(Type=Bone,Name=\"spine_02\"),Weight=0.0368),(Item=(Type=Bone,Name=\"neck_02\"),Weight=0.0641),(Item=(Type=Bone,Name=\"deform_pec_r\"),Weight=0.0812),(Item=(Type=Bone,Name=\"clavicle_r\"),Weight=0.095),(Ite ...(truncated)"
    Weight=1
   )
   > &RigUnit_ParentConstraint(
    AdvancedSettings=(RotationOrderForFilter=XZY)
    Child=(Type=Bone,Name="base_tipScarf_bck")
    Parents="((Item=(Type=Bone,Name=\"deform_upperarm_up_r\"),Weight=0.0031),(Item=(Type=Bone,Name=\"spine_04\"),Weight=0.0512),(Item=(Type=Bone,Name=\"dyn_scarf_bck\"),Weight=0.0529),(Item=(Type=Bone,Name=\"neck_02\"),Weight=0.0544),(Item=(Type=Bone,Name=\"clavicle_l\"),Weight= ...(truncated)"
    Weight=1
   )
   Multiply = &RigVMFunction_MathDoubleMul(B=RigVMVariableNode() {variable=aimDown}.Value, A=RigVMVariableNode() {variable=meeleWeapon}.Value)
   > &RigUnit_ModifyTransforms(
    ItemToModify=((Item=(Type=Bone,Name="base_tipScarf_m"),Transform=(Rotation=(X=0,Y=0,Z=0,W=1),Translation=(X=3.4323,Y=0,Z=7.2261),Scale3D=(X=1,Y=1,Z=1))))
    WeightMinimum=0
    Weight=@Multiply.Result
   )
   n1 = RigVMVariableNode() {variable=floatingVar}
   > &RigUnit_ModifyTransforms(
    ItemToModify="((Item=(Type=Bone,Name=\"None\"),Transform=(Rotation=(X=0,Y=0,Z=0,W=1),Translation=(X=0,Y=0,Z=0),Scale3D=(X=1,Y=1,Z=1))),(Item=(Type=Bone,Name=\"dyn_scarf_bck\"),Transform=(Rotation=(X=0,Y=0,Z=0,W=-1),Translation=(X=11.3647,Y=0,Z=-14.0178),Scale3D=(X=1,Y=1,Z=1 ...(truncated)"
    WeightMinimum=0
    Weight=&RigVMFunction_MathDoubleMul(
     A=&RigVMFunction_MathDoubleRemap(SourceMaximum=1, SourceMinimum=0, TargetMaximum=0, TargetMinimum=1, bClamp=False, Value=@n1.Value).Result
     B=RigVMVariableNode() {variable=LookUp}.Value
    ).Result
   )
   > RigVMFunctionReturnNode()
  ]}
 RigVMFunctionLibrary {nodes [
   BeltSystem = RigVMCollapseNode(ExecuteContext=())
   Control Setup = RigVMCollapseNode(Execute Content=(), debug=False)
   DeltaRotation_X = RigVMCollapseNode(ExecuteContext=())
   DeltaRotation_Y = RigVMCollapseNode(ExecuteContext=())
   DeltaRotation_Z = RigVMCollapseNode(ExecuteContext=())
   PadsSystem = RigVMCollapseNode(ExecuteContext=())
   Scarf System = RigVMCollapseNode(ExecuteContext=())
   SkirtSystem = RigVMCollapseNode(ExecuteContext=())
   Variables = RigVMCollapseNode(ExecuteContext=())
  ]}
 RigVMFunction_Sequence_1_ContainedGraph {nodes [
   n1 = RigVMFunctionEntryNode()
   RigVMFunction_Sequence = &RigVMFunction_Sequence(ExecuteContext=@n1.ExecuteContext)
   n2 = &RigVMFunction_Sequence(ExecuteContext=@RigVMFunction_Sequence.B)
   RigVMFunctionReturnNode(A=@RigVMFunction_Sequence.A, B=@n2.A, C=@n2.B)
  ]}
 RigVMFunction_Sequence_1_ContainedGraph {nodes [
   n1 = RigVMFunctionEntryNode()
   RigVMFunction_Sequence = &RigVMFunction_Sequence(ExecuteContext=@n1.ExecuteContext)
   n2 = &RigVMFunction_Sequence(ExecuteContext=@RigVMFunction_Sequence.B)
   n3 = &RigVMFunction_Sequence(ExecuteContext=@n2.B)
   RigVMFunctionReturnNode(A=@RigVMFunction_Sequence.A, B=@n2.A, C=@n3.A, D=@n3.B)
  ]}
 RigVMFunction_Sequence_1_ContainedGraph {nodes [
   n1 = RigVMFunctionEntryNode()
   RigVMFunction_Sequence = &RigVMFunction_Sequence(ExecuteContext=@n1.ExecuteContext)
   n2 = &RigVMFunction_Sequence(ExecuteContext=@RigVMFunction_Sequence.B)
   RigVMFunctionReturnNode(A=@RigVMFunction_Sequence.A, B=@n2.A, C=@n2.B)
  ]}
 RigVMModel {nodes [
   RigVMFunction_Sequence = RigVMAggregateNode(ExecuteContext=@RigUnit_BeginExecution.ExecutePin)
   Control Setup = RigVMFunctionReferenceNode(debug=false, Execute Content=@RigVMFunction_Sequence.A)
   DeltaRotation_Z = RigVMFunctionReferenceNode(Child=dyn_skirt_D_01, Parent=base_skirt_D, Source Maximum=-100, ExecuteContext=@"Control Setup".Execute Content)
   RigUnit_BeginExecution = &RigUnit_BeginExecution()
   DISPATCH_RigVMDispatch_Print = RigVMDispatchNode(Enabled=True, ExecuteContext=(), ScreenColor=(R=1,G=1,B=1,A=1), ScreenDuration=0.05, Value=0) {template="DISPATCH_RigVMDispatch_Print(in Prefix,in Value,in Enabled,in ScreenDuration,in ScreenColor)"}
   Scarf System = RigVMFunctionReferenceNode(ExecuteContext=@BeltSystem.ExecuteContext)
   RigVMVariableNode() {variable=thighSideRotateR}
   BeltSystem = RigVMFunctionReferenceNode(ExecuteContext=@SkirtSystem.ExecuteContext)
   FN_DeformRig_FNC = RigVMFunctionReferenceNode(
    DeformRig_ Curve="(deformRig_l_upperarm_up_alpha_72_#1=true,deformRig_r_upperarm_up_alpha_73_#2=true,deformRig_l_upperarm_fr_alpha_74_#3=true,deformRig_r_upperarm_fr_alpha_75_CB51B80C4 ...(truncated)"
    DeformRig_Bool="(upperarm_1_#4=True,elbow_3_#5=True,wrist_7_#6=True,groin_8_#7=True,glute_17_#8=false,knee_9_40C408324CA26 ...(truncated)"
    DeformRig_Debug=(upperarm_debug_18_#4=false,wrist_debug_20_#6=false,lat_debug_26_#9=false,groinGlute_debug_27_#7=false)
    DeformRig_Value="(upperarm_up_18_#10=(Rotation=(X=0.0,Y=-0.0,Z=0.0,W=1.0),Translation=(X=0.0,Y=0.0,Z=8.0),Scale3D=(X=0.70,Y=1.0,Z=1.0)),upperarm_fwd_fr_19_#11=(Rotation=(X=0.0,Y=-0.0,Z=0.0,W=1.0),Translation=(X=0.0, ...(truncated)"
    ExecuteContext=@RigVMFunction_Sequence.B
   )
   > Variables = RigVMFunctionReferenceNode()
   > PadsSystem = RigVMFunctionReferenceNode()
   > SkirtSystem = RigVMFunctionReferenceNode()
  ]}
 SequenceAggregate_3_1_ContainedGraph {nodes [
   n1 = RigVMFunctionEntryNode()
   SequenceAggregate_3 = &RigVMFunction_Sequence(ExecuteContext=@n1.ExecuteContext)
   SequenceAggregate_3_1 = &RigVMFunction_Sequence(ExecuteContext=@SequenceAggregate_3.B)
   SequenceAggregate_3_2 = &RigVMFunction_Sequence(ExecuteContext=@SequenceAggregate_3_1.B)
   SequenceAggregate_3_3 = &RigVMFunction_Sequence(ExecuteContext=@SequenceAggregate_3_2.B)
   RigVMFunctionReturnNode(A=@SequenceAggregate_3.A, B=@SequenceAggregate_3_1.A, C=@SequenceAggregate_3_2.A, D=@SequenceAggregate_3_3.A, E=@SequenceAggregate_3_3.B)
  ]}
]
hierarchy [
 root {loc=<0 0 0>, rot=<0 0 0 1>}
 pelvis {parent=root, loc=<0 -0.7071 93.1191>, rot=<0.0127 0.707 -0.0127 -0.707>}
 spine_01 {parent=pelvis, loc=<6.7494 -0.6388 0>, rot=<0 0 0.034 -0.9994>}
 spine_02 {parent=spine_01, loc=<7.2518 -0.035 0>, rot=<0 0 0.0238 -0.9997>}
 spine_03 {parent=spine_02, loc=<6.0104 0 0>, rot=<0 0 0.0149 -0.9999>}
 spine_04 {parent=spine_03, loc=<8.1024 0 0>, rot=<0 0 -0.0132 -0.9999>}
 spine_05 {parent=spine_04, loc=<10.0146 0 0>, rot=<0 0 -0.0015 -1>}
 attach_cape {parent=spine_05, loc=<9.0204 -11.6963 0.0012>, rot=<0.0616 -0.7044 -0.0616 -0.7044>}
 attach_backpack {parent=spine_05, loc=<-0.1186 -13.4505 0.0006>, rot=<-0.0093 -0.707 0.0093 -0.707>}
 clavicle_l {parent=spine_05, loc=<4.3575 -0.1466 -2.3223>, rot=<0.0465 -0.7289 -0.009 -0.6829>}
 upperarm_l {parent=clavicle_l, loc=<15.5111 0 0>, rot=<0.0226 -0.3234 -0.0088 -0.9459>}
 lowerarm_l {parent=upperarm_l, loc=<26.6514 0 0>, rot=<0.0391 -0.0099 -0.2453 -0.9686>}
 hand_l {parent=lowerarm_l, loc=<22.3491 0 0>, rot=<-0.7081 -0.0507 0.0975 -0.6975>}
 thumb_01_l {parent=hand_l, loc=<1.3118 -0.3191 -2.459>, rot=<0.6546 -0.4042 0.0664 -0.6354>}
 thumb_02_l {parent=thumb_01_l, loc=<5.9336 -0.0448 0.0401>, rot=<-0.0037 0.0213 0.298 -0.9543>}
 thumb_03_l {parent=thumb_02_l, loc=<2.2223 0.3682 -0.0907>, rot=<0.0118 -0.013 0.0066 -0.9998>}
 index_metacarpal_l {parent=hand_l, loc=<1.6535 0.0183 -2.96>, rot=<0.0158 -0.0124 -0.0173 -0.9996>}
 index_01_l {parent=index_metacarpal_l, loc=<8.3008 0.0357 -0.0524>, rot=<-0.0044 0.0485 0.0924 -0.9945>}
 index_02_l {parent=index_01_l, loc=<3.3587 -0.3004 0.0452>, rot=<-0.002 0.0027 0.3225 -0.9466>}
 index_03_l {parent=index_02_l, loc=<2.5578 -0.123 -0.0252>, rot=<-0.002 0.0067 0.2558 -0.9667>}
 middle_metacarpal_l {parent=hand_l, loc=<2.1856 -0.1725 -0.5984>, rot=<0.006 0.0168 -0.0329 -0.9993>}
 middle_01_l {parent=middle_metacarpal_l, loc=<7.3275 -0.2447 0.0129>, rot=<-0.0208 0.0042 0.0858 -0.9961>}
 middle_02_l {parent=middle_01_l, loc=<4.0123 0.1401 0.1175>, rot=<-0.0218 0.0258 0.3512 -0.9357>}
 middle_03_l {parent=middle_02_l, loc=<2.6077 0.0534 0.1373>, rot=<0.0082 0.0188 0.1708 -0.9851>}
 ring_metacarpal_l {parent=hand_l, loc=<2.2074 -0.3625 1.3638>, rot=<-0.0284 0.0722 -0.0458 -0.9959>}
 ring_01_l {parent=ring_metacarpal_l, loc=<7.0295 -0.6197 -0.1446>, rot=<-0.0489 -0.0252 0.1892 -0.9804>}
 ring_02_l {parent=ring_01_l, loc=<3.6174 0.4647 -0.1027>, rot=<0 0 0.2806 -0.9598>}
 ring_03_l {parent=ring_02_l, loc=<2.4152 0.114 0.0199>, rot=<0 0 0.1189 -0.9929>}
 pinky_metacarpal_l {parent=hand_l, loc=<1.7616 -0.9753 3.3814>, rot=<-0.0397 0.0814 -0.069 -0.9935>}
 pinky_01_l {parent=pinky_metacarpal_l, loc=<7.1709 -0.4831 -0.1094>, rot=<-0.0654 -0.0035 0.2242 -0.9723>}
 pinky_02_l {parent=pinky_01_l, loc=<2.8274 0.0893 -0.1043>, rot=<0.0037 0.0011 0.2854 -0.9584>}
 pinky_03_l {parent=pinky_02_l, loc=<2.2046 0.0307 -0.0078>, rot=<-0.0002 0.0098 0.1445 -0.9895>}
 deform_wrist_dn_base_l {parent=hand_l, loc=<0 -2.2342 0>, rot=<0 0 0 -1>}
 deform_wrist_dn_l {parent=deform_wrist_dn_base_l, loc=<0 0 0>, rot=<0 0 0 1>}
 deform_wrist_up_base_l {parent=hand_l, loc=<0 2.2682 0>, rot=<0 0 0 -1>}
 deform_wrist_up_l {parent=deform_wrist_up_base_l, loc=<0 0 0>, rot=<0 0 0 1>}
 lowerarm_twist_02_l {parent=lowerarm_l, loc=<9.6199 0 0>, rot=<0 0 0 -1>}
 base_vine01_l {parent=lowerarm_twist_02_l, loc=<-1.6825 0.9763 8.4389>, rot=<0.0514 0.6901 -0.0415 -0.7207>}
 dyn_vine01_l {parent=base_vine01_l, loc=<0 0 0>, rot=<0 0 0 1>}
 base_vine02_l {parent=lowerarm_twist_02_l, loc=<-2.8139 7.3259 -2.5586>, rot=<0.171 -0.1881 0.0093 -0.9671>}
 dyn_vine02_l {parent=base_vine02_l, loc=<0 0 0>, rot=<0 0 0 1>}
 lowerarm_twist_01_l {parent=lowerarm_l, loc=<17.8549 0 0>, rot=<0 0 0 -1>}
 deform_elbow_in_base_l {parent=lowerarm_l, loc=<0.2112 5.259 0.8274>, rot=<0 0 0 -1>}
 deform_elbow_in_l {parent=deform_elbow_in_base_l, loc=<0 0 0>, rot=<0 0 0 1>}
 deform_elbow_out_base_l {parent=lowerarm_l, loc=<0 -4 0>, rot=<0 0 0 -1>}
 deform_elbow_out_l {parent=deform_elbow_out_base_l, loc=<0 0 0>, rot=<0 0 0 1>}
 upperarm_twist_01_l {parent=upperarm_l, loc=<0.0041 0 0>, rot=<0 0 0 1>}
 deform_upperarm_bk_base_l {parent=upperarm_twist_01_l, loc=<1.0813 -7.2193 -1.9453>, rot=<0 0 0 1>}
 deform_upperarm_bk_l {parent=deform_upperarm_bk_base_l, loc=<0 0 0>, rot=<0 0 0 1>}
 deform_upperarm_fr_base_l {parent=upperarm_twist_01_l, loc=<-0.3691 7.0506 -2.765>, rot=<0 0 0 1>}
 deform_upperarm_fr_l {parent=deform_upperarm_fr_base_l, loc=<0 0 0>, rot=<0 0 0 1>}
 deform_upperarm_up_base_l {parent=upperarm_twist_01_l, loc=<-3.2216 -1.1391 3.3311>, rot=<0 0 0 1>}
 deform_upperarm_up_l {parent=deform_upperarm_up_base_l, loc=<0 0 0>, rot=<0 0 0 1>}
 base_shoulderPad_l {parent=upperarm_twist_01_l, loc=<-0.0041 0 0>, rot=<0 0 0 1>}
 dyn_shoulderPad_l {parent=base_shoulderPad_l, loc=<0 0 0>, rot=<0 0 0 1>}
 base_shoulderPDM_l {parent=dyn_shoulderPad_l, loc=<5.6336 0 5.3217>, rot=<0 0.1156 0 -0.9933>}
 dyn_shoulderPDM01_l {parent=base_shoulderPDM_l, loc=<0 0 0>, rot=<0 0 0 1>}
 dyn_shoulderPDM02_l {parent=dyn_shoulderPDM01_l, loc=<9.3767 -0.9765 0>, rot=<0 0.24 0 -0.9708>}
 upperarm_twist_02_l {parent=upperarm_l, loc=<13.7941 0 0>, rot=<0 0 0 1>}
 neck_01 {parent=spine_05, loc=<11.3138 0.1529 0>, rot=<0 0 -0.2003 -0.9797>}
 neck_02 {parent=neck_01, loc=<4.5627 0 0>, rot=<0 0 0.0504 -0.9987>}
 head {parent=neck_02, loc=<5.3914 0 0>, rot=<0 0 0.115 -0.9934>}
 base_scarf_bck {parent=neck_02, loc=<-1.0195 -10.9973 0>, rot=<-0.0784 -0.7027 0.7027 0.0784>}
 dyn_scarf_bck {parent=base_scarf_bck, loc=<0 0 0>, rot=<0 0 0 1>}
 base_tipScarf_bck {parent=neck_01, loc=<-4.2627 -14.3387 0>, rot=<0 -0.7071 0.7071 0>}
 dyn_tipScarf_bck {parent=base_tipScarf_bck, loc=<0 0 0>, rot=<0 0 0 1>}
 clavicle_r {parent=spine_05, loc=<4.3575 -0.1466 2.3223>, rot=<0.7289 0.0465 -0.6829 0.009>}
 upperarm_r {parent=clavicle_r, loc=<-15.5111 0 0>, rot=<0.0226 -0.3234 -0.0088 -0.9459>}
 lowerarm_r {parent=upperarm_r, loc=<-26.6514 0 0>, rot=<0.0391 -0.0099 -0.2453 -0.9686>}
 hand_r {parent=lowerarm_r, loc=<-22.3491 0 0>, rot=<-0.7081 -0.0507 0.0975 -0.6975>}
 thumb_01_r {parent=hand_r, loc=<-1.3118 0.3191 2.459>, rot=<0.6546 -0.4042 0.0664 -0.6354>}
 thumb_02_r {parent=thumb_01_r, loc=<-5.9336 0.0448 -0.0401>, rot=<-0.0037 0.0213 0.298 -0.9543>}
 thumb_03_r {parent=thumb_02_r, loc=<-2.2223 -0.3682 0.0907>, rot=<0.0118 -0.013 0.0066 -0.9998>}
 index_metacarpal_r {parent=hand_r, loc=<-1.6535 -0.0183 2.96>, rot=<0.0158 -0.0124 -0.0173 -0.9996>}
 index_01_r {parent=index_metacarpal_r, loc=<-8.3008 -0.0357 0.0524>, rot=<-0.0044 0.0485 0.0924 -0.9945>}
 index_02_r {parent=index_01_r, loc=<-3.3587 0.3004 -0.0452>, rot=<-0.002 0.0027 0.3225 -0.9466>}
 index_03_r {parent=index_02_r, loc=<-2.5578 0.123 0.0252>, rot=<-0.002 0.0067 0.2558 -0.9667>}
 middle_metacarpal_r {parent=hand_r, loc=<-2.1856 0.1725 0.5984>, rot=<0.006 0.0168 -0.0329 -0.9993>}
 middle_01_r {parent=middle_metacarpal_r, loc=<-7.3275 0.2447 -0.0129>, rot=<-0.0208 0.0042 0.0858 -0.9961>}
 middle_02_r {parent=middle_01_r, loc=<-4.0123 -0.1401 -0.1175>, rot=<-0.0218 0.0258 0.3512 -0.9357>}
 middle_03_r {parent=middle_02_r, loc=<-2.6077 -0.0534 -0.1373>, rot=<0.0082 0.0188 0.1708 -0.9851>}
 ring_metacarpal_r {parent=hand_r, loc=<-2.2074 0.3625 -1.3638>, rot=<-0.0284 0.0722 -0.0458 -0.9959>}
 ring_01_r {parent=ring_metacarpal_r, loc=<-7.0295 0.6197 0.1446>, rot=<-0.0489 -0.0252 0.1892 -0.9804>}
 ring_02_r {parent=ring_01_r, loc=<-3.6174 -0.4647 0.1027>, rot=<0 0 0.2806 -0.9598>}
 ring_03_r {parent=ring_02_r, loc=<-2.4152 -0.114 -0.0199>, rot=<0 0 0.1189 -0.9929>}
 pinky_metacarpal_r {parent=hand_r, loc=<-1.7616 0.9753 -3.3814>, rot=<-0.0397 0.0814 -0.069 -0.9935>}
 pinky_01_r {parent=pinky_metacarpal_r, loc=<-7.1709 0.4831 0.1094>, rot=<-0.0654 -0.0035 0.2242 -0.9723>}
 pinky_02_r {parent=pinky_01_r, loc=<-2.8274 -0.0893 0.1043>, rot=<0.0037 0.0011 0.2854 -0.9584>}
 pinky_03_r {parent=pinky_02_r, loc=<-2.2046 -0.0307 0.0078>, rot=<-0.0002 0.0098 0.1445 -0.9895>}
 deform_wrist_dn_base_r {parent=hand_r, loc=<-0.0004 2.2345 -0.0001>, rot=<0 0 0 -1>}
 deform_wrist_dn_r {parent=deform_wrist_dn_base_r, loc=<0 0 0>, rot=<0 0 0 1>}
 deform_wrist_up_base_r {parent=hand_r, loc=<0.0003 -2.2684 0.0001>, rot=<0 0 0 -1>}
 deform_wrist_up_r {parent=deform_wrist_up_base_r, loc=<0 0 0>, rot=<0 0 0 1>}
 lowerarm_twist_02_r {parent=lowerarm_r, loc=<-9.6199 0 0>, rot=<0 0 0 -1>}
 base_vine02_r {parent=lowerarm_twist_02_r, loc=<-3.2069 1.9478 -6.0536>, rot=<-0.1535 -0.7539 -0.1341 -0.6246>}
 dyn_vine02_r {parent=base_vine02_r, loc=<0 0 0>, rot=<0 0 0 1>}
 lowerarm_twist_01_r {parent=lowerarm_r, loc=<-17.8549 0 0>, rot=<0 0 0 -1>}
 deform_elbow_in_base_r {parent=lowerarm_r, loc=<-0.2111 -5.2591 -0.8275>, rot=<0 0 0 -1>}
 deform_elbow_in_r {parent=deform_elbow_in_base_r, loc=<-0.0001 0 0.0002>, rot=<0 0 0 1>}
 deform_elbow_out_base_r {parent=lowerarm_r, loc=<0 4 -0.0001>, rot=<0 0 0 -1>}
 deform_elbow_out_r {parent=deform_elbow_out_base_r, loc=<0 0 0>, rot=<0 0 0 1>}
 upperarm_twist_01_r {parent=upperarm_r, loc=<-0.0041 0 0>, rot=<0 0 0 -1>}
 deform_upperarm_bk_base_r {parent=upperarm_twist_01_r, loc=<-1.0814 7.2192 1.9455>, rot=<0 0 0 -1>}
 deform_upperarm_bk_r {parent=deform_upperarm_bk_base_r, loc=<0 0 0>, rot=<0 0 0 1>}
 deform_upperarm_fr_base_r {parent=upperarm_twist_01_r, loc=<0.3689 -7.0506 2.7652>, rot=<0 0 0 -1>}
 deform_upperarm_fr_r {parent=deform_upperarm_fr_base_r, loc=<0 0 0>, rot=<0 0 0 1>}
 deform_upperarm_up_base_r {parent=upperarm_twist_01_r, loc=<3.2214 1.1391 -3.3308>, rot=<0 0 0 1>}
 deform_upperarm_up_r {parent=deform_upperarm_up_base_r, loc=<0 0 0>, rot=<0 0 0 -1>}
 base_shoulderPad_r {parent=upperarm_twist_01_r, loc=<0.0041 0 0>, rot=<0 0 0 -1>}
 dyn_shoulderPad_r {parent=base_shoulderPad_r, loc=<0 0 0>, rot=<0 0 0 1>}
 base_vine01_r {parent=dyn_shoulderPad_r, loc=<-1.1705 -1.5553 -14.6206>, rot=<0.0268 -0.727 0.0561 -0.6838>}
 dyn_vine01_r {parent=base_vine01_r, loc=<0 0 0>, rot=<0 0 0 1>}
 upperarm_twist_02_r {parent=upperarm_r, loc=<-13.7941 0 0>, rot=<0 0 0 -1>}
 base_tipScarf_m {parent=spine_05, loc=<6.2183 15.9907 3.3696>, rot=<-0.0284 0.7065 0.7065 -0.0284>}
 dyn_tipScarf_m {parent=base_tipScarf_m, loc=<0 0 0>, rot=<0 0 0 1>}
 base_scarf_fr {parent=spine_05, loc=<9.0944 14.2694 0>, rot=<-0.0284 0.7065 0.7065 -0.0284>}
 dyn_scarf_fr {parent=base_scarf_fr, loc=<0 0 0>, rot=<0 0 0 1>}
 deform_pec_aim_base_l {parent=spine_04, loc=<7.8732 12.6388 -2.4123>, rot=<0.0102 -0.4602 0.3979 -0.7936>}
 deform_pec_aim_l {parent=deform_pec_aim_base_l, loc=<0 0 0>, rot=<0 0 0 1>}
 deform_pec_base_l {parent=deform_pec_aim_l, loc=<10.766 0 0>, rot=<0 0 0 1>}
 deform_pec_l {parent=deform_pec_base_l, loc=<0.055 0 0>, rot=<0 0 0 1>}
 deform_pec_aim_base_r {parent=spine_04, loc=<7.8727 12.6388 2.4123>, rot=<-0.4602 -0.0102 0.7936 0.3979>}
 deform_pec_aim_r {parent=deform_pec_aim_base_r, loc=<0 0 0>, rot=<0 0 0 1>}
 deform_pec_base_r {parent=deform_pec_aim_r, loc=<-10.7662 -0.0006 -0.0007>, rot=<0 0 0 1>}
 deform_pec_r {parent=deform_pec_base_r, loc=<0 0 0>, rot=<0 0 0 1>}
 thigh_l {parent=pelvis, loc=<-2.5693 -0.0109 -7.7501>, rot=<0.0012 0.0572 0.0155 -0.9982>}
 calf_l {parent=thigh_l, loc=<-39.0484 0 0>, rot=<0.0006 -0.0161 -0.0391 -0.9991>}
 foot_l {parent=calf_l, loc=<-42.561 0.1077 0.4292>, rot=<0.0418 -0.0455 0.0343 -0.9975>}
 ball_l {parent=foot_l, loc=<-6.1599 13.8264 0.6806>, rot=<0.0001 -0.0001 -0.7063 -0.7079>}
 calf_twist_02_l {parent=calf_l, loc=<-18.8064 -0.0967 0.0397>, rot=<0.0237 0.0095 -0.0035 -0.9997>}
 base_vine03_l {parent=calf_twist_02_l, loc=<-2.2042 0.3949 -6.4922>, rot=<0.0989 -0.7367 0.1453 -0.6529>}
 dyn_vine03_l {parent=base_vine03_l, loc=<0 0 0>, rot=<0 0 0 1>}
 calf_twist_01_l {parent=calf_l, loc=<-35.813 -0.0967 0.0397>, rot=<0.0237 0.0095 -0.0035 -0.9997>}
 deform_knee_in_base_l {parent=calf_l, loc=<1.298 -5.2957 -0.0496>, rot=<0 0 0 -1>}
 deform_knee_in_l {parent=deform_knee_in_base_l, loc=<0 0 0>, rot=<0 0 0 1>}
 deform_knee_out_base_l {parent=calf_l, loc=<0 4 0>, rot=<0 0 0 -1>}
 deform_knee_out_l {parent=deform_knee_out_base_l, loc=<0 0 0>, rot=<0 0 0 1>}
 thigh_twist_01_l {parent=thigh_l, loc=<-21.6303 0 0>, rot=<0.0013 0 0 -1>}
 base_pant01_l {parent=thigh_twist_01_l, loc=<-6.6845 2.9066 7.8126>, rot=<0.0572 -0.0015 0.9984 0.0024>}
 dyn_pant01_l {parent=base_pant01_l, loc=<0 0 0>, rot=<0 0 0 1>}
 base_pant02_l {parent=thigh_twist_01_l, loc=<-13.1778 -5.9852 3.1482>, rot=<0.0495 -0.4662 0.8828 0.0287>}
 dyn_pant02_l {parent=base_pant02_l, loc=<0 0 0>, rot=<0 0 0 1>}
 deform_groin_base_l {parent=thigh_l, loc=<-2.1908 8.1142 -1.7442>, rot=<0 0 0 -1>}
 deform_groin_l {parent=deform_groin_base_l, loc=<0 0 0>, rot=<0 0 0 1>}
 base_kneepad_l {parent=thigh_l, loc=<-39.0484 0 0>, rot=<0.0006 -0.0161 -0.0391 -0.9991>}
 base_outbelt_fr_l {parent=thigh_l, loc=<-1.7977 11.9347 -0.693>, rot=<0.0421 0.7058 0.7061 -0.0387>}
 dyn_outbelt_fr_l {parent=base_outbelt_fr_l, loc=<0 0 0>, rot=<0 0 0 1>}
 thigh_r {parent=pelvis, loc=<-2.5693 -0.0109 7.7501>, rot=<0.0572 -0.0012 0.9982 0.0155>}
 calf_r {parent=thigh_r, loc=<39.0484 0 0>, rot=<0.0006 -0.0161 -0.0391 -0.9991>}
 foot_r {parent=calf_r, loc=<42.561 -0.1077 -0.4292>, rot=<0.0418 -0.0455 0.0343 -0.9975>}
 ball_r {parent=foot_r, loc=<6.1599 -13.8264 -0.6806>, rot=<0.0001 -0.0001 -0.7063 -0.7079>}
 calf_twist_02_r {parent=calf_r, loc=<18.8064 0.0967 -0.0397>, rot=<0.0237 0.0095 -0.0035 -0.9997>}
 calf_twist_01_r {parent=calf_r, loc=<35.813 0.0967 -0.0397>, rot=<0.0237 0.0095 -0.0035 -0.9997>}
 deform_knee_in_base_r {parent=calf_r, loc=<-1.298 5.2957 0.0496>, rot=<0 0 0 1>}
 deform_knee_in_r {parent=deform_knee_in_base_r, loc=<0 0 0>, rot=<0 0 0 1>}
 deform_knee_out_base_r {parent=calf_r, loc=<0 -4 0>, rot=<0 0 0 -1>}
 deform_knee_out_r {parent=deform_knee_out_base_r, loc=<0 0 0>, rot=<0 0 0 1>}
 thigh_twist_01_r {parent=thigh_r, loc=<21.6303 0 0>, rot=<0.0013 0 0 -1>}
 base_pant01_r {parent=thigh_twist_01_r, loc=<6.7368 0.2358 -7.4816>, rot=<0.9984 -0.0024 -0.0572 -0.0015>}
 dyn_pant01_r {parent=base_pant01_r, loc=<0 0 0>, rot=<0 0 0 1>}
 base_pant02_r {parent=thigh_twist_01_r, loc=<9.0389 5.4469 1.2189>, rot=<0.5957 -0.0472 -0.0323 -0.8012>}
 dyn_pant02_r {parent=base_pant02_r, loc=<0 0 0>, rot=<0 0 0 1>}
 deform_groin_base_r {parent=thigh_r, loc=<2.1908 -8.1142 1.7442>, rot=<0 0 0 -1>}
 deform_groin_r {parent=deform_groin_base_r, loc=<0 0 0>, rot=<0 0 0 1>}
 base_kneepad_r {parent=thigh_r, loc=<39.0484 0 0>, rot=<0.0006 -0.0161 -0.0391 -0.9991>}
 deform_glute_base_l {parent=pelvis, loc=<-8.8064 -8.8222 -8.4287>, rot=<0.0275 0.0501 -0.4472 -0.8926>}
 deform_glute_l {parent=deform_glute_base_l, loc=<0 0 0>, rot=<0 0 0 1>}
 deform_glute_base_r {parent=pelvis, loc=<-8.8062 -8.8222 8.4287>, rot=<0.0501 -0.0275 0.8926 -0.4472>}
 deform_glute_r {parent=deform_glute_base_r, loc=<0 0 0>, rot=<0 0 0 1>}
 base_skirt_D {parent=pelvis, loc=<3.156 -13.1206 -0.0012>, rot=<0.0904 -0.9959 0.0009 0.0001>}
 dyn_skirt_D_01 {parent=base_skirt_D, loc=<0 0 0>, rot=<0 0 0 1>}
 dyn_skirt_D_02 {parent=dyn_skirt_D_01, loc=<14.6156 0 0>, rot=<0 0 0 1>}
 dyn_skirt_D_03 {parent=dyn_skirt_D_02, loc=<13.0665 0 0>, rot=<0 0 0 1>}
 dyn_skirt_D_04 {parent=dyn_skirt_D_03, loc=<12.4735 0 0>, rot=<0 0 0 1>}
 base_skirt_G {parent=pelvis, loc=<1.7925 10.7013 6.5112>, rot=<0.1403 -0.1602 0.9743 0.0729>}
 dyn_skirt_G_01 {parent=base_skirt_G, loc=<0 0 0>, rot=<0 0 0 1>}
 dyn_skirt_G_02 {parent=dyn_skirt_G_01, loc=<8.694 0 0>, rot=<0 0 0 1>}
 dyn_skirt_G_03 {parent=dyn_skirt_G_02, loc=<9.4534 0 0>, rot=<0 0 0 1>}
 dyn_skirt_G_04 {parent=dyn_skirt_G_03, loc=<8.8447 0.0001 -0.0001>, rot=<0 0 0 1>}
 dyn_skirt_G_05 {parent=dyn_skirt_G_04, loc=<12.5 0 0>, rot=<0 0 0 1>}
 base_skirt_F {parent=pelvis, loc=<2.3972 -0.368 15.1884>, rot=<0.0834 -0.6816 0.7228 0.078>}
 dyn_skirt_F_01 {parent=base_skirt_F, loc=<0 0 0>, rot=<0 0 0 1>}
 dyn_skirt_F_02 {parent=dyn_skirt_F_01, loc=<13.0476 0 0>, rot=<0 0 0 1>}
 dyn_skirt_F_03 {parent=dyn_skirt_F_02, loc=<11.5891 0 0>, rot=<0 0 0 1>}
 dyn_skirt_F_04 {parent=dyn_skirt_F_03, loc=<12.4052 0 0>, rot=<0 0 0 1>}
 base_skirt_E {parent=pelvis, loc=<3.2413 -9.2376 10.738>, rot=<0.0951 -0.9302 0.3504 0.0547>}
 dyn_skirt_E_01 {parent=base_skirt_E, loc=<0 0 0>, rot=<0 0 0 1>}
 dyn_skirt_E_02 {parent=dyn_skirt_E_01, loc=<13.6834 0 0>, rot=<0 0 0 1>}
 dyn_skirt_E_03 {parent=dyn_skirt_E_02, loc=<11.8039 -0.0001 0>, rot=<0 0 0 1>}
 dyn_skirt_E_04 {parent=dyn_skirt_E_03, loc=<12.593 0.0001 0>, rot=<0 0 0 1>}
 base_skirt_A {parent=pelvis, loc=<1.7926 10.7013 -6.5111>, rot=<-0.1399 0.1657 0.9735 0.072>}
 dyn_skirt_A_01 {parent=base_skirt_A, loc=<0 0 0>, rot=<0 0 0 1>}
 dyn_skirt_A_02 {parent=dyn_skirt_A_01, loc=<8.694 0 0>, rot=<0 0 0 1>}
 dyn_skirt_A_03 {parent=dyn_skirt_A_02, loc=<9.4534 0 0>, rot=<0 0 0 1>}
 dyn_skirt_A_04 {parent=dyn_skirt_A_03, loc=<8.8447 0 0>, rot=<0 0 0 1>}
 dyn_skirt_A_05 {parent=dyn_skirt_A_04, loc=<13.7888 1.1836 0>, rot=<0 0 -0.0514 -0.9987>}
 base_skirt_B {parent=pelvis, loc=<2.3972 -0.368 -15.1884>, rot=<-0.0836 0.6815 0.7227 0.0783>}
 dyn_skirt_B_01 {parent=base_skirt_B, loc=<0 0 0>, rot=<0 0 0 1>}
 dyn_skirt_B_02 {parent=dyn_skirt_B_01, loc=<13.0476 0 0>, rot=<0 0 0 1>}
 dyn_skirt_B_03 {parent=dyn_skirt_B_02, loc=<11.589 0 0>, rot=<0 0 0 1>}
 dyn_skirt_B_04 {parent=dyn_skirt_B_03, loc=<12.4053 0 0>, rot=<0 0 0 1>}
 base_skirt_C {parent=pelvis, loc=<2.8215 -10.3207 -11.716>, rot=<0.0969 -0.93 -0.3503 -0.0554>}
 dyn_skirt_C_01 {parent=base_skirt_C, loc=<0 0 0>, rot=<0 0 0 1>}
 dyn_skirt_C_02 {parent=dyn_skirt_C_01, loc=<13.6834 0 0>, rot=<0 0 0 1>}
 dyn_skirt_C_03 {parent=dyn_skirt_C_02, loc=<11.804 0 0>, rot=<0 0 0 1>}
 dyn_skirt_C_04 {parent=dyn_skirt_C_03, loc=<12.593 0 0>, rot=<0 0 0 1>}
 base_inbelt_l {parent=pelvis, loc=<6.9234 7.5985 -8.8224>, rot=<0.0127 0.707 0.707 0.0127>}
 dyn_inbelt_l {parent=base_inbelt_l, loc=<0 0 0>, rot=<0 0 0 1>}
 base_inbelt_r {parent=pelvis, loc=<6.9234 7.5985 8.8224>, rot=<0.0127 0.707 0.707 0.0127>}
 dyn_inbelt_r {parent=base_inbelt_r, loc=<0 0 0>, rot=<0 0 0 -1>}
 base_beltBag_l {parent=pelvis, loc=<6.8756 2.9734 -15.3677>, rot=<-0.0178 -0.9919 -0.1258 -0.0023>}
 dyn_beltBag_l {parent=base_beltBag_l, loc=<0 0 0>, rot=<0 0 0 1>}
 base_outbelt_r {parent=pelvis, loc=<4.0406 -0.5417 8.8452>, rot=<0 0 0.9998 0.0179>}
 dyn_outbelt_r {parent=base_outbelt_r, loc=<0 0 0>, rot=<0 0 0 1>}
 base_feather_d01_r {parent=dyn_outbelt_r, loc=<0.4845 1.1276 6.7838>, rot=<0.0283 0.2124 0.0016 -0.9768>}
 dyn_feather_d01_r {parent=base_feather_d01_r, loc=<0 0 0>, rot=<0 0 0 1>}
 base_feather_c01_r {parent=dyn_outbelt_r, loc=<1.0048 -5.7354 5.7902>, rot=<-0.1891 0.2075 -0.0455 -0.9587>}
 dyn_feather_c01_r {parent=base_feather_c01_r, loc=<0 0 0>, rot=<0 0 0 1>}
 base_feather_b01_r {parent=dyn_outbelt_r, loc=<2.1923 -10.2556 2.2059>, rot=<-0.4455 0.1877 -0.0621 -0.8732>}
 dyn_feather_b01_r {parent=base_feather_b01_r, loc=<0 0 0>, rot=<0 0 0 1>}
 dyn_feather_b02_r {parent=dyn_feather_b01_r, loc=<8.8335 0 0>, rot=<0 0 0 1>}
 base_feather_a01_r {parent=dyn_outbelt_r, loc=<4.3045 -12.5398 -2.0071>, rot=<-0.7062 0.1089 -0.0367 -0.6987>}
 dyn_feather_a01_r {parent=base_feather_a01_r, loc=<0 0 0>, rot=<0 0 0 1>}
 dyn_feather_a02_r {parent=dyn_feather_a01_r, loc=<8.8335 0 0>, rot=<0 0 0 1>}
 base_feather_e01_r {parent=dyn_outbelt_r, loc=<1.0579 6.2294 6.9402>, rot=<0.2461 0.1104 -0.0282 -0.9625>}
 dyn_feather_e01_r {parent=base_feather_e01_r, loc=<0 0 0>, rot=<0 0 0 1>}
 base_outbelt_bck_l {parent=pelvis, loc=<-6.1401 -13.8088 -9.3968>, rot=<-0.0127 -0.707 0.707 0.0127>}
 dyn_outbelt_bck_l {parent=base_outbelt_bck_l, loc=<0 0 0>, rot=<0 0 0 1>}
 base_outbelt_side_l {parent=pelvis, loc=<-8.093 -1.5244 -18.2013>, rot=<-0.0178 -0.9942 0.0019 -0.1059>}
 dyn_outbelt_side_l {parent=base_outbelt_side_l, loc=<0 0 0>, rot=<0 0 0 1>}
 upperarm_l_ctrl {controlType=EULER_TRANSFORM, type=Control, loc=<0 0 0>, rot=<0 0 0 1>}
 thigh_r_ctrl {controlType=EULER_TRANSFORM, type=Control, loc=<0 0 0>, rot=<0 0 0 1>}
 thigh_l_ctrl {controlType=EULER_TRANSFORM, type=Control, loc=<0 0 0>, rot=<0 0 0 1>}
 dyn_skirt_D_01_ctrl {controlType=EULER_TRANSFORM, type=Control, loc=<0 0 0>, rot=<0 0 0 1>}
]
variables [
 IsRunFloat {type=double, default=0}
 LookDown {type=double, default=0}
 LookUp {type=double, default=0}
 NotInAir {type=double, default=0}
 RunCrouch {type=double, default=0}
 aimDown {type=double, default=0}
 floatingVar {type=double, default=0}
 idleVar {type=double, default=0}
 isCrouching {type=bool, default=False}
 isRunning {type=bool, default=False}
 isSliding {type=double, default=0}
 meeleWeapon {type=double, default=0}
 moving2d {type=double, default=0}
 thighSideRotateL {type=double, default=0}
 thighSideRotateR {type=double, default=0}
]
