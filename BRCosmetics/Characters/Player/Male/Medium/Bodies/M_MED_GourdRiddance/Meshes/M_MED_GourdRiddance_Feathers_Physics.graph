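# Physics asset for the GourdRiddance male medium body cosmetic, with capsule bodies for right-side feathers, scarf tips and skirt bones.
# Feather and scarf dyn bones chain from base bones with limited swing and little or no twist, and both scarf bases are angle-locked to spine_05.
# Skirt G's three-bone chain has free angular motion, while skirt E, thigh_r and the feather bases have no constraint to a parent body.
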PhysicsAsset /BRCosmetics/Characters/Player/Male/Medium/Bodies/M_MED_GourdRiddance/Meshes/M_MED_GourdRiddance_Feathers_Physics
bodies {
 base_feather_a01_r [{shapeType=Capsule, length=0.505, radius=0.505, shapeName=base_feather_a01_r_capsule, rotation=<rot 90 0 0>}]
 base_feather_b01_r [{shapeType=Capsule, length=0.505, radius=0.505, shapeName=base_feather_b01_r_capsule, rotation=<rot 90 0 0>}]
 base_feather_c01_r [{shapeType=Capsule, length=0.505, radius=0.505, shapeName=base_feather_c01_r_capsule, rotation=<rot 90 0 0>}]
 base_feather_d01_r [{shapeType=Capsule, length=0.505, radius=0.505, shapeName=base_feather_d01_r_capsule, rotation=<rot 90 0 0>}]
 base_feather_e01_r [{shapeType=Capsule, length=0.505, radius=0.505, shapeName=base_feather_e01_r_capsule, rotation=<rot 0 -90 0>}]
 base_tipScarf_bck [{shapeType=Capsule, length=11.9098, radius=3.9745, shapeName=base_tipScarf_bck_capsule, center=<6.2365 0 -7.5293>, rotation=<rot 0 0 90>}]
 base_tipScarf_m [{shapeType=Capsule, length=6.6487, radius=6.711, shapeName=base_tipScarf_m_capsule, center=<5.8697 0 -7.4417>, rotation=<rot 0 0 90>}]
 dyn_feather_a01_r [{shapeType=Capsule, length=4.1597, radius=3.2783, shapeName=dyn_feather_a01_r_capsule, center=<4.6163 0 1.8352>, rotation=<rot 90 0 0>}]
 dyn_feather_a02_r [{shapeType=Capsule, length=4.6352, radius=2.2156, shapeName=dyn_feather_a02_r_capsule, center=<4.6163 0 1.7707>, rotation=<rot 90 0 0>}]
 dyn_feather_b01_r [{shapeType=Capsule, length=5.0512, radius=2.2942, shapeName=dyn_feather_b01_r_capsule, center=<4.7958 0 2.8213>, rotation=<rot 90 0 0>}]
 dyn_feather_b02_r [{shapeType=Capsule, length=3.079, radius=2.8489, shapeName=dyn_feather_b02_r_capsule, center=<4.7958 0 1.4501>, rotation=<rot 90 0 0>}]
 dyn_feather_c01_r [{
   shapeType=Capsule
   length=6.5247
   radius=3.1692
   shapeName=dyn_feather_c01_r_capsule
   center=<6.7446 0.1144 1.838>
   rotation=<rot 89.4581 0.3539 0.3609>
  }]
 dyn_feather_d01_r [{
   shapeType=Capsule
   length=4.7166
   radius=3.0393
   shapeName=dyn_feather_d01_r_capsule
   center=<3.708 -0.7262 0.8099>
   rotation=<rot 90 179.156 179.1562>
  }]
 dyn_feather_e01_r [{
   shapeType=Capsule
   length=5.6779
   radius=2.3032
   shapeName=dyn_feather_e01_r_capsule
   center=<6.16 -0.0909 0.3019>
   rotation=<rot 0 82.6189 -86.6195>
  }]
 dyn_skirt_E_01 [{
   shapeType=Capsule
   length=6.4387
   radius=6.1855
   shapeName=dyn_skirt_E_01_capsule
   center=<7.3077 6.0501 -4.7279>
   rotation=<rot 5.5834 -90 -0.5628>
  }]
 dyn_skirt_G_01 [{shapeType=Capsule, length=7.9429, radius=3.0571, shapeName=dyn_skirt_G_01_capsule, center=<0 1.5602 0>}]
 dyn_skirt_G_02 [{shapeType=Capsule, length=7.9429, radius=3.0571, shapeName=dyn_skirt_G_02_capsule, center=<0 2.3005 0>}]
 dyn_skirt_G_03 [{shapeType=Capsule, length=7.9429, radius=3.0571, shapeName=dyn_skirt_G_03_capsule, center=<0 2.3005 0>}]
 dyn_tipScarf_bck [{shapeType=Capsule, length=7.6067, radius=2.1266, shapeName=dyn_tipScarf_bck_capsule, center=<8.1701 0 0>, rotation=<rot 90 0 0>}]
 dyn_tipScarf_m [{shapeType=Capsule, length=6.4274, radius=1.7012, shapeName=dyn_tipScarf_m_capsule, center=<4.2893 0 0.3923>, rotation=<rot 90 0 0>}]
 spine_05 [{shapeType=Capsule, length=16.6801, radius=13.4354, shapeName=spine_05_capsule}]
 thigh_r [{
   shapeType=Capsule
   length=34.5583
   radius=12.4417
   shapeName=thigh_l_capsule
   center=<20.2278 -0.2477 -2.5317>
   rotation=<rot 84.375 179.9998 -0.0001>
  }]
}
constraints [
 {bone1Name=dyn_feather_a01_r, bone2Name=base_feather_a01_r, swing1LimitDegrees=10, swing2LimitDegrees=10, twistLimitDegrees=5}
 {bone1Name=dyn_feather_a02_r, bone2Name=dyn_feather_a01_r, swing1LimitDegrees=10, swing2LimitDegrees=25, twistLimitDegrees=0}
 {bone1Name=dyn_feather_b01_r, bone2Name=base_feather_b01_r, swing1LimitDegrees=15, swing2LimitDegrees=10, twistLimitDegrees=5}
 {bone1Name=dyn_feather_b02_r, bone2Name=dyn_feather_b01_r, swing1LimitDegrees=10, swing2LimitDegrees=15, twistLimitDegrees=0}
 {bone1Name=dyn_feather_c01_r, bone2Name=base_feather_c01_r, swing1LimitDegrees=30, swing2LimitDegrees=25, twistLimitDegrees=0}
 {bone1Name=dyn_feather_d01_r, bone2Name=base_feather_d01_r, swing1LimitDegrees=30, swing2LimitDegrees=25, twistLimitDegrees=0}
 {bone1Name=dyn_tipScarf_m, bone2Name=base_tipScarf_m, swing1LimitDegrees=25, swing2LimitDegrees=20, twistLimitDegrees=0}
 {
  bone1Name=base_tipScarf_m
  bone2Name=spine_05
  swing1LimitDegrees=45
  swing1Motion=Locked
  swing2LimitDegrees=45
  swing2Motion=Locked
  twistLimitDegrees=45
  twistMotion=Locked
 }
 {bone1Name=dyn_feather_e01_r, bone2Name=base_feather_e01_r, swing1LimitDegrees=30, swing2LimitDegrees=25, twistLimitDegrees=0}
 {bone1Name=dyn_tipScarf_bck, bone2Name=base_tipScarf_bck, swing1LimitDegrees=25, swing2LimitDegrees=50, twistLimitDegrees=0}
 {
  bone1Name=base_tipScarf_bck
  bone2Name=spine_05
  swing1LimitDegrees=45
  swing1Motion=Locked
  swing2LimitDegrees=45
  swing2Motion=Locked
  twistLimitDegrees=45
  twistMotion=Locked
 }
 {
  bone1Name=dyn_skirt_G_03
  bone2Name=dyn_skirt_G_02
  swing1LimitDegrees=45
  swing1Motion=Free
  swing2LimitDegrees=45
  swing2Motion=Free
  twistLimitDegrees=45
  twistMotion=Free
 }
 {
  bone1Name=dyn_skirt_G_02
  bone2Name=dyn_skirt_G_01
  swing1LimitDegrees=45
  swing1Motion=Free
  swing2LimitDegrees=45
  swing2Motion=Free
  twistLimitDegrees=45
  twistMotion=Free
 }
]
