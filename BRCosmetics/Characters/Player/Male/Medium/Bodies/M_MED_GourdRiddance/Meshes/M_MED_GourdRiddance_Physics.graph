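# Physics asset for the male medium GourdRiddance player body cosmetic, giving its seven skirt bone chains (A–G) capsule bodies and joints.
# Each chain runs from the root through joints with limited swing and small or locked twist; anchors are locked except chain A's free base.
# Swing1 limits widen toward the hem on C–E and narrow on B, F and G.
# Leg, foot, left kneepad and root capsules add body volumes, including a huge one below the root.
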PhysicsAsset /BRCosmetics/Characters/Player/Male/Medium/Bodies/M_MED_GourdRiddance/Meshes/M_MED_GourdRiddance_Physics
bodies {
 base_kneepad_l [{
   shapeType=Capsule
   length=18.1068
   radius=4.626
   shapeName=base_kneepad_l_capsule
   center=<-0.8493 11.8882 -1.0251>
   rotation=<rot 0 105.8576 -88.2577>
  }]
 base_skirt_A [{shapeType=Capsule, length=0.505, radius=0.505, shapeName=base_skirt_A_capsule, rotation=<rot 0 -90 0>}]
 base_skirt_C [{shapeType=Capsule, length=1, radius=1, shapeName=base_skirt_C_capsule}]
 base_skirt_D [{shapeType=Capsule, length=1, radius=1, shapeName=base_skirt_D_capsule}]
 base_skirt_E [{shapeType=Capsule, length=1, radius=1, shapeName=base_skirt_E_capsule}]
 calf_l [{shapeType=Capsule, length=30, radius=10, shapeName=calf_l_capsule, center=<-20.9198 0 0>, rotation=<rot 90 0 0>}]
 calf_r [{shapeType=Capsule, length=31.6416, radius=10.8584, shapeName=calf_l_capsule, center=<20.9198 0 0>, rotation=<rot 90 0 180>}]
 dyn_skirt_A_01 [{shapeType=Capsule, length=6, radius=2, shapeName=dyn_skirt_A_01_capsule, center=<4.7294 0 0>, rotation=<rot 90 0 0>}]
 dyn_skirt_A_02 [{shapeType=Capsule, length=4.8054, radius=2, shapeName=dyn_skirt_A_02_capsule, center=<5.0437 0.5934 -0.0822>, rotation=<rot 90 0 0>}]
 dyn_skirt_A_03 [{shapeType=Capsule, length=5.7901, radius=3.5, shapeName=dyn_skirt_A_03_capsule, center=<4.1397 -0.4736 -1.1947>, rotation=<rot 90 0 0>}]
 dyn_skirt_A_04 [{
   shapeType=Capsule
   length=7.3947
   radius=4
   shapeName=dyn_skirt_A_04_capsule
   center=<6.2172 -0.4746 -2.6467>
   rotation=<rot 86.2743 -0.0003 -0.0003>
  }]
 dyn_skirt_A_05 [{shapeType=Capsule, length=10.5182, radius=4.1307, shapeName=dyn_skirt_A_05_capsule, center=<9.6564 1.2536 -4.1515>, rotation=<rot 90 0 0>}]
 dyn_skirt_B_01 [{shapeType=Capsule, length=1, radius=1, shapeName=dyn_skirt_B_01_capsule}]
 dyn_skirt_B_02 [{shapeType=Capsule, length=6.5, radius=2.5, shapeName=dyn_skirt_B_02_capsule, center=<9 0 0>}]
 dyn_skirt_B_03 [{shapeType=Capsule, length=11, radius=4, shapeName=dyn_skirt_B_03_capsule, center=<7.0288 0 0>, rotation=<rot 0 0 -5.625>}]
 dyn_skirt_B_04 [{shapeType=Capsule, length=11, radius=4, shapeName=dyn_skirt_B_04_capsule, center=<12.2864 -1.343 -0.3364>, rotation=<rot 0.0001 0 -14.0625>}]
 dyn_skirt_C_01 [{shapeType=Capsule, length=5, radius=3, shapeName=dyn_skirt_C_01_capsule, center=<6.4853 0 0>}]
 dyn_skirt_C_02 [{shapeType=Capsule, length=6.5, radius=3, shapeName=dyn_skirt_C_02_capsule, center=<7.2451 0 0>}]
 dyn_skirt_C_03 [{shapeType=Capsule, length=7.5, radius=4, shapeName=dyn_skirt_C_03_capsule, center=<7.8078 0 0>}]
 dyn_skirt_C_04 [{shapeType=Capsule, length=8.5, radius=4.5, shapeName=dyn_skirt_C_04_capsule, center=<12.2864 -2.1882 0>}]
 dyn_skirt_D_01 [{shapeType=Capsule, length=5, radius=3, shapeName=dyn_skirt_D_01_capsule, center=<6.4853 0 0>}]
 dyn_skirt_D_02 [{shapeType=Capsule, length=7, radius=3, shapeName=dyn_skirt_D_02_capsule, center=<7.0202 0 0>}]
 dyn_skirt_D_03 [{shapeType=Capsule, length=6, radius=4, shapeName=dyn_skirt_D_03_capsule, center=<7.8078 0 0>}]
 dyn_skirt_D_04 [{shapeType=Capsule, length=7, radius=4.5, shapeName=dyn_skirt_D_04_capsule, center=<12.2864 0 0>}]
 dyn_skirt_E_01 [{shapeType=Capsule, length=5, radius=3, shapeName=dyn_skirt_E_01_capsule, center=<6.4853 0 0>}]
 dyn_skirt_E_02 [{shapeType=Capsule, length=6.5, radius=3, shapeName=dyn_skirt_E_02_capsule, center=<7.1526 -1.7603 0>}]
 dyn_skirt_E_03 [{shapeType=Capsule, length=7.5, radius=4, shapeName=dyn_skirt_E_03_capsule, center=<7.8078 0 0>}]
 dyn_skirt_E_04 [{shapeType=Capsule, length=6.3449, radius=5.9631, shapeName=dyn_skirt_E_04_capsule, center=<12.2864 0 0.5574>, rotation=<rot 0 0 3.2085>}]
 dyn_skirt_F_01 [{shapeType=Capsule, length=1, radius=1, shapeName=dyn_skirt_F_01_capsule}]
 dyn_skirt_F_02 [{shapeType=Capsule, length=8, radius=2.5, shapeName=dyn_skirt_F_02_capsule, center=<6.0406 -1.2355 0>}]
 dyn_skirt_F_03 [{shapeType=Capsule, length=7.1412, radius=4, shapeName=dyn_skirt_F_03_capsule, center=<9.4443 -1.9004 0>}]
 dyn_skirt_F_04 [{shapeType=Capsule, length=7.1412, radius=4.5, shapeName=dyn_skirt_F_04_capsule, center=<14.1849 -1.7868 0>}]
 dyn_skirt_G_01 [{shapeType=Capsule, length=1, radius=1, shapeName=dyn_skirt_G_01_capsule}]
 dyn_skirt_G_02 [
  {shapeType=Capsule, length=4.1237, radius=3.5, shapeName=dyn_skirt_G_02_capsule, center=<4.9155 -1.7147 0>, rotation=<rot 90 0 0>}
  {shapeType=Capsule, length=4.1237, radius=3.5, shapeName=dyn_skirt_G_02_capsule, center=<4.9155 -0.2052 4.4297>, rotation=<rot 90 0 0>}
 ]
 dyn_skirt_G_03 [{shapeType=Capsule, length=5.0442, radius=4, shapeName=dyn_skirt_G_03_capsule, center=<5.4162 -1.6982 0>, rotation=<rot 90 0 0>}]
 dyn_skirt_G_04 [{shapeType=Capsule, length=1, radius=5.5, shapeName=dyn_skirt_G_04_capsule, center=<9.1641 -1.2609 0.9957>, rotation=<rot 90 0 0>}]
 dyn_skirt_G_05 [{
   shapeType=Capsule
   length=7.884
   radius=4.3008
   shapeName=dyn_skirt_G_05_capsule
   center=<12.2224 0.3172 -0.3718>
   rotation=<rot 0 0 49.5348>
  }]
 foot_l [{shapeType=Capsule, length=13.5, radius=11, shapeName=foot_l_capsule, center=<-4.577 6.4941 0>, rotation=<rot 90 180 -89.9998>}]
 foot_r [{shapeType=Capsule, length=13.5, radius=11, shapeName=foot_l_capsule, center=<4.577 -6.4941 0>, rotation=<rot 90 -165.9638 104.0364>}]
 root [
  {shapeType=Capsule, length=15, radius=15, shapeName=root_capsule, center=<0 11.9719 23.3001>}
  {shapeType=Capsule, length=70, radius=150, shapeName=root_capsule, center=<0 0 -141.8158>, rotation=<rot 0 0 90>}
 ]
 thigh_l [{
   shapeType=Capsule
   length=35.5
   radius=11.5
   shapeName=thigh_l_capsule
   center=<-20.2137 -1.8221 2.3882>
   rotation=<rot 84.375 179.9996 179.9998>
  }]
 thigh_r [{
   shapeType=Capsule
   length=35.5
   radius=11.5
   shapeName=thigh_l_capsule
   center=<20.2678 1.7026 -2.9377>
   rotation=<rot 84.375 179.9998 -0.0001>
  }]
}
constraints [
 {
  bone1Name=dyn_skirt_B_01
  bone2Name=root
  swing1LimitDegrees=45
  swing1Motion=Locked
  swing2LimitDegrees=10
  swing2Motion=Locked
  twistLimitDegrees=5
  twistMotion=Locked
 }
 {bone1Name=dyn_skirt_B_02, bone2Name=dyn_skirt_B_01, swing1LimitDegrees=35, swing2LimitDegrees=20, twistLimitDegrees=5}
 {bone1Name=dyn_skirt_B_03, bone2Name=dyn_skirt_B_02, swing1LimitDegrees=25, swing2LimitDegrees=20, twistLimitDegrees=5, twistMotion=Locked}
 {bone1Name=dyn_skirt_B_04, bone2Name=dyn_skirt_B_03, swing1LimitDegrees=15, swing2LimitDegrees=15, twistLimitDegrees=5, twistMotion=Locked}
 {
  bone1Name=base_skirt_C
  bone2Name=root
  swing1LimitDegrees=45
  swing1Motion=Locked
  swing2LimitDegrees=45
  swing2Motion=Locked
  twistLimitDegrees=45
  twistMotion=Locked
 }
 {
  bone1Name=dyn_skirt_C_01
  bone2Name=base_skirt_C
  swing1LimitDegrees=5
  swing1Motion=Locked
  swing2LimitDegrees=10
  twistLimitDegrees=5
  twistMotion=Locked
 }
 {bone1Name=dyn_skirt_C_02, bone2Name=dyn_skirt_C_01, swing1LimitDegrees=12.5, swing2LimitDegrees=10, twistLimitDegrees=5}
 {bone1Name=dyn_skirt_C_03, bone2Name=dyn_skirt_C_02, swing1LimitDegrees=25, swing2LimitDegrees=5, twistLimitDegrees=5, twistMotion=Locked}
 {bone1Name=dyn_skirt_C_04, bone2Name=dyn_skirt_C_03, swing1LimitDegrees=45, swing2LimitDegrees=1, twistLimitDegrees=5, twistMotion=Locked}
 {
  bone1Name=base_skirt_D
  bone2Name=root
  swing1LimitDegrees=45
  swing1Motion=Locked
  swing2LimitDegrees=45
  swing2Motion=Locked
  twistLimitDegrees=45
  twistMotion=Locked
 }
 {bone1Name=dyn_skirt_D_01, bone2Name=base_skirt_D, swing1LimitDegrees=5, swing2LimitDegrees=1, twistLimitDegrees=1}
 {bone1Name=dyn_skirt_D_02, bone2Name=dyn_skirt_D_01, swing1LimitDegrees=10, swing2LimitDegrees=1, twistLimitDegrees=5}
 {bone1Name=dyn_skirt_D_03, bone2Name=dyn_skirt_D_02, swing1LimitDegrees=25, swing2LimitDegrees=1, twistLimitDegrees=5, twistMotion=Locked}
 {bone1Name=dyn_skirt_D_04, bone2Name=dyn_skirt_D_03, swing1LimitDegrees=45, swing2LimitDegrees=1, twistLimitDegrees=5, twistMotion=Locked}
 {
  bone1Name=base_skirt_E
  bone2Name=root
  swing1LimitDegrees=45
  swing1Motion=Locked
  swing2LimitDegrees=45
  swing2Motion=Locked
  twistLimitDegrees=45
  twistMotion=Locked
 }
 {bone1Name=dyn_skirt_E_01, bone2Name=base_skirt_E, swing1LimitDegrees=5, swing2LimitDegrees=10, twistLimitDegrees=5, twistMotion=Locked}
 {bone1Name=dyn_skirt_E_02, bone2Name=dyn_skirt_E_01, swing1LimitDegrees=10, swing2LimitDegrees=10, twistLimitDegrees=5}
 {bone1Name=dyn_skirt_E_03, bone2Name=dyn_skirt_E_02, swing1LimitDegrees=25, swing2LimitDegrees=5, twistLimitDegrees=5, twistMotion=Locked}
 {bone1Name=dyn_skirt_E_04, bone2Name=dyn_skirt_E_03, swing1LimitDegrees=45, swing2LimitDegrees=1, twistLimitDegrees=5, twistMotion=Locked}
 {
  bone1Name=dyn_skirt_F_01
  bone2Name=root
  swing1LimitDegrees=45
  swing1Motion=Locked
  swing2LimitDegrees=10
  swing2Motion=Locked
  twistLimitDegrees=5
  twistMotion=Locked
 }
 {bone1Name=dyn_skirt_F_02, bone2Name=dyn_skirt_F_01, swing1LimitDegrees=35, swing2LimitDegrees=15, twistLimitDegrees=5}
 {bone1Name=dyn_skirt_F_03, bone2Name=dyn_skirt_F_02, swing1LimitDegrees=25, swing2LimitDegrees=5, twistLimitDegrees=5, twistMotion=Locked}
 {bone1Name=dyn_skirt_F_04, bone2Name=dyn_skirt_F_03, swing1LimitDegrees=15, swing2LimitDegrees=5, twistLimitDegrees=5, twistMotion=Locked}
 {
  bone1Name=dyn_skirt_G_01
  bone2Name=root
  swing1LimitDegrees=65
  swing1Motion=Locked
  swing2LimitDegrees=35
  swing2Motion=Locked
  twistLimitDegrees=5
  twistMotion=Locked
 }
 {bone1Name=dyn_skirt_G_02, bone2Name=dyn_skirt_G_01, swing1LimitDegrees=30, swing2LimitDegrees=5, twistLimitDegrees=5, twistMotion=Locked}
 {bone1Name=dyn_skirt_G_03, bone2Name=dyn_skirt_G_02, swing1LimitDegrees=15, swing2LimitDegrees=5, twistLimitDegrees=5, twistMotion=Locked}
 {bone1Name=dyn_skirt_G_04, bone2Name=dyn_skirt_G_03, swing1LimitDegrees=10, swing2LimitDegrees=5, twistLimitDegrees=5, twistMotion=Locked}
 {
  bone1Name=thigh_r
  bone2Name=root
  swing1LimitDegrees=45
  swing1Motion=Locked
  swing2LimitDegrees=45
  swing2Motion=Locked
  twistLimitDegrees=45
  twistMotion=Locked
 }
 {
  bone1Name=calf_l
  bone2Name=thigh_l
  swing1LimitDegrees=45
  swing1Motion=Locked
  swing2LimitDegrees=45
  swing2Motion=Locked
  twistLimitDegrees=45
  twistMotion=Locked
 }
 {
  bone1Name=calf_r
  bone2Name=thigh_r
  swing1LimitDegrees=45
  swing1Motion=Locked
  swing2LimitDegrees=45
  swing2Motion=Locked
  twistLimitDegrees=45
  twistMotion=Locked
 }
 {
  bone1Name=foot_l
  bone2Name=thigh_l
  swing1LimitDegrees=45
  swing1Motion=Locked
  swing2LimitDegrees=45
  swing2Motion=Locked
  twistLimitDegrees=45
  twistMotion=Locked
 }
 {
  bone1Name=foot_l
  bone2Name=calf_l
  swing1LimitDegrees=45
  swing1Motion=Locked
  swing2LimitDegrees=45
  swing2Motion=Locked
  twistLimitDegrees=45
  twistMotion=Locked
 }
 {
  bone1Name=thigh_l
  bone2Name=root
  swing1LimitDegrees=45
  swing1Motion=Locked
  swing2LimitDegrees=45
  swing2Motion=Locked
  twistLimitDegrees=45
  twistMotion=Locked
 }
 {
  bone1Name=foot_r
  bone2Name=calf_r
  swing1LimitDegrees=45
  swing1Motion=Locked
  swing2LimitDegrees=45
  swing2Motion=Locked
  twistLimitDegrees=45
  twistMotion=Locked
 }
 {
  bone1Name=base_kneepad_l
  bone2Name=thigh_l
  swing1LimitDegrees=45
  swing1Motion=Free
  swing2LimitDegrees=45
  swing2Motion=Free
  twistLimitDegrees=45
  twistMotion=Free
 }
 {bone1Name=dyn_skirt_G_05, bone2Name=dyn_skirt_G_04, swing1LimitDegrees=10, swing2LimitDegrees=20, twistLimitDegrees=0}
 {
  bone1Name=base_skirt_A
  bone2Name=root
  swing1LimitDegrees=45
  swing1Motion=Free
  swing2LimitDegrees=45
  swing2Motion=Free
  twistLimitDegrees=45
  twistMotion=Free
 }
 {bone1Name=dyn_skirt_A_01, bone2Name=base_skirt_A, swing1LimitDegrees=30, swing2LimitDegrees=30, twistLimitDegrees=0}
 {bone1Name=dyn_skirt_A_02, bone2Name=dyn_skirt_A_01, swing1LimitDegrees=15, swing2LimitDegrees=25, twistLimitDegrees=0}
 {bone1Name=dyn_skirt_A_03, bone2Name=dyn_skirt_A_02, swing1LimitDegrees=15, swing2LimitDegrees=30, twistLimitDegrees=0}
 {bone1Name=dyn_skirt_A_04, bone2Name=dyn_skirt_A_03, swing1LimitDegrees=10, swing2LimitDegrees=30, twistLimitDegrees=0, twistMotion=Locked}
 {bone1Name=dyn_skirt_A_05, bone2Name=dyn_skirt_A_04, swing1LimitDegrees=0, swing2LimitDegrees=10, twistLimitDegrees=0}
]
