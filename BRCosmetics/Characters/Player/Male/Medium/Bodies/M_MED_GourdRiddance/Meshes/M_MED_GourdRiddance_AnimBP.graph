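# Post-process style Animation Blueprint for the GourdRiddance body cosmetic, adding secondary motion to skirt, feathers, vines, pants and scarf.
# The parent mesh's pose passes through a deform Control Rig and head offset, then AnimDynamics and feather RigidBody simulation.
# Cached branches add state-specific physics or hand-posed Look At and bone offsets, blended for emotes, baseball sliding, sliding and default.
# The update event derives head counter-rotation, gravity override and weapon state.
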AnimBlueprint /BRCosmetics/Characters/Player/Male/Medium/Bodies/M_MED_GourdRiddance/Meshes/M_MED_GourdRiddance_AnimBP
target_skeleton=&/M_MED_GourdRiddance_Skeleton
graphs [
 AnimGraph {nodes [
   n1 = "Transform (Modify) Bone - Bone: base_outbelt_r"!(
    ComponentPose=Reroute Node(InputPin:Component Space Pose Link Structure) -> (OutputPin:Component Space Pose Link Structure)
    bAlphaBoolEnabled:bool=True
    Translation="-11.905,0.476,1.200"
    Rotation="-0.000,38.660,-0.000"
   ) -> (Pose) {
    alphaInputType=Bool
    rotationMode=BMM_Additive
    rotationSpace=BCS_BoneSpace
    translationMode=BMM_Additive
    translationSpace=BCS_BoneSpace
    alphaBoolBlend {blendOption=Linear}
    boneToModify {boneName=base_outbelt_r}
    rotation=<rot 0 38.66 0>
    translation=<-11.905 0.476 1.2>
   }
   n2 = "Transform (Modify) Bone - Bone: base_outbelt_bck_l"!(ComponentPose=@n1, bAlphaBoolEnabled:bool=True, Translation="-15,0,0.85", Rotation="0,0,0") -> (Pose) {
    alphaInputType=Bool
    rotationMode=BMM_Additive
    rotationSpace=BCS_BoneSpace
    translationMode=BMM_Additive
    translationSpace=BCS_BoneSpace
    alphaBoolBlend {blendOption=Linear}
    boneToModify {boneName=base_outbelt_bck_l}
    translation {x=-15, z=0.85}
   }
   n3 = "Transform (Modify) Bone - Bone: base_outbelt_side_l"!(ComponentPose=@n2, bAlphaBoolEnabled:bool=True, Translation="-13.6,-0.06,3", Rotation="0.37,-29,0.45") -> (Pose) {
    alphaInputType=Bool
    rotationMode=BMM_Additive
    rotationSpace=BCS_BoneSpace
    translationMode=BMM_Additive
    translationSpace=BCS_BoneSpace
    alphaBoolBlend {blendOption=Linear}
    boneToModify {boneName=base_outbelt_side_l}
    rotation=<rot 0.37 -29 0.45>
    translation=<-13.6 -0.06 3>
   }
   n4 = "Constraint - base_feather_d01_r (PPPPPP)"!(
    ComponentPose=@n3
    Alpha=1
    ConstraintWeights_0=0.5772
    ConstraintWeights_1:float=0.4
    ConstraintWeights_2:float=0.0197
    ConstraintWeights_3:float=0.0013
    ConstraintWeights_4:float=0.0011
    ConstraintWeights_5:float=0.0006
   ) -> (Pose) {
    alphaBoolBlend {blendOption=Linear}
    boneToModify {boneName=base_feather_d01_r}
    constraintSetup [
     {targetBone {boneName=pelvis}}
     {targetBone {boneName=dyn_outbelt_r}}
     {targetBone {boneName=deform_glute_r}}
     {targetBone {boneName=spine_02}}
     {targetBone {boneName=spine_01}}
     {targetBone {boneName=spine_03}}
    ]
    constraintWeights [1,1,1,1,1,1]
   }
   n5 = "Constraint - base_feather_c01_r (PPPPP)"!(
    ComponentPose=@n4
    Alpha=1
    ConstraintWeights_0=0.5783
    ConstraintWeights_1:float=0.4
    ConstraintWeights_2:float=0.0095
    ConstraintWeights_3:float=0.0078
    ConstraintWeights_4:float=0.0044
   ) -> (Pose) {
    alphaBoolBlend {blendOption=Linear}
    boneToModify {boneName=base_feather_c01_r}
    constraintSetup [
     {targetBone {boneName=pelvis}}
     {targetBone {boneName=dyn_outbelt_r}}
     {targetBone {boneName=spine_02}}
     {targetBone {boneName=spine_01}}
     {targetBone {boneName=spine_03}}
    ]
    constraintWeights [1,1,1,1,1]
   }
   n6 = "Constraint - base_feather_b01_r (PPPPPP)"!(
    ComponentPose=@n5
    Alpha=1
    ConstraintWeights_0=0.591
    ConstraintWeights_1:float=0.4
    ConstraintWeights_2:float=0.0044
    ConstraintWeights_3:float=0.0026
    ConstraintWeights_4:float=0.0015
    ConstraintWeights_5:float=0.0005
   ) -> (Pose) {
    alphaBoolBlend {blendOption=Linear}
    boneToModify {boneName=base_feather_b01_r}
    constraintSetup [
     {targetBone {boneName=pelvis}}
     {targetBone {boneName=dyn_outbelt_r}}
     {targetBone {boneName=spine_02}}
     {targetBone {boneName=spine_01}}
     {targetBone {boneName=spine_03}}
     {targetBone {boneName=thigh_twist_01_l}}
    ]
    constraintWeights [1,1,1,1,1,1]
   }
   "Constraint - base_feather_a01_r (PPPPPPP)"!(
    ComponentPose=@n6
    Alpha=1
    ConstraintWeights_0=0.5721
    ConstraintWeights_1:float=0.4
    ConstraintWeights_2:float=0.0152
    ConstraintWeights_3:float=0.0067
    ConstraintWeights_4:float=0.0036
    ConstraintWeights_5:float=0.0015
    ConstraintWeights_6:float=0.0009
   ) -> (Pose) {
    alphaBoolBlend {blendOption=Linear}
    boneToModify {boneName=base_feather_a01_r}
    constraintSetup [
     {targetBone {boneName=pelvis}}
     {targetBone {boneName=dyn_outbelt_r}}
     {targetBone {boneName=thigh_twist_01_l}}
     {targetBone {boneName=deform_groin_l}}
     {targetBone {boneName=spine_02}}
     {targetBone {boneName=spine_01}}
     {targetBone {boneName=spine_03}}
    ]
    constraintWeights [1,1,1,1,1,1,1]
   }
   Copy Pose From Mesh() {bUseAttachedParent=true}
   "Transform (Modify) Bone - Bone: base_feather_a01_r"!(ComponentPose, bAlphaBoolEnabled:bool=True, Translation="0,0,1", Rotation="20,0,0", Scale="1,1,1") -> (Pose) {
    alphaInputType=Bool
    rotationMode=BMM_Additive
    rotationSpace=BCS_BoneSpace
    translationMode=BMM_Additive
    translationSpace=BCS_BoneSpace
    alphaBoolBlend {blendOption=Linear}
    boneToModify {boneName=base_feather_a01_r}
    rotation {pitch=20}
    translation {z=1}
   }
   n7 = Blend Poses by bool(
    BlendPose_0=Use cached pose 'CRCache_SlidingDisable'()
    BlendPose_1=Blend Poses by bool(
     BlendPose_0=Use cached pose 'CRCache_Sliding'()
     BlendPose_1=Use cached pose 'CRCache_SkirtBase_Setup'()
     BlendTime_0=0.5
     bActiveValue=Get SlidingRBEnable() -> (SlidingRBEnable:bool)
    ) {blendType=Linear}
    BlendTime_0=0.5
    bActiveValue=Get SlidingRBdisabled() -> (SlidingRBdisabled:bool)
   ) {blendType=Linear}
   n8 = Use cached pose 'CRCache_BaseBallSliding'()
   n9 = Get bIsBaseballSliding()
   Output Pose(Result=Blend Poses by bool(
     BlendPose_0=Use cached pose 'CRCache_Emote'()
     BlendPose_1=Blend Poses by bool(BlendPose_0=@n8, BlendPose_1=@n7, BlendTime_0=0.5, bActiveValue=@n9) {blendType=Linear}
     BlendTime_0=0.5
     BlendTime_1=0.3
     bActiveValue=Get bIsPlayingEmote()
    ) {blendType=Linear}) {layerGroup=None}
   n10 = Control Rig!(
    Source=Copy Pose From Mesh() {bUseAttachedParent=true}
    Alpha
    floatingVar:double=Get floatingVar() -> (floatingVar:double)
    moving2d:double=Get Moving2d() -> (Moving2d:double)
    isCrouching:bool=Get bIsCrouching()
    isRunning:bool=Get IsRunning() -> (IsRunning:bool)
    isSliding:double=Get isSliding() -> (isSliding:double)
    IsRunFloat:double=Get isRunFloat() -> (isRunFloat:double)
    meeleWeapon:double=Get MeeleWeapon() -> (MeeleWeapon:float)
   ) -> (Pose) {
    alphaBoolBlend {blendOption=Linear}
    controlRigAssetReference {blueprintRigClass=&/M_MED_GourdRiddance_Deform_CtrlRig.M_MED_GourdRiddance_Deform_CtrlRig_C}
   }
   n11 = "Transform (Modify) Bone - Bone: head"!(ComponentPose=Local To Component(LocalPose=@n10), Alpha=1, Rotation=Get head rot() -> (head rot:Rotator)) -> (Pose) {rotationMode=BMM_Additive, rotationSpace=BCS_BoneSpace, alphaBoolBlend {blendOption=Linear}, boneToModify {boneName=head}}
   Save cached pose 'CRCache_InputPose'(Pose=Component To Local(ComponentPose=@n11)) {cacheName=CRCache_InputPose}
   n12 = "Transform (Modify) Bone - Bone: dyn_skirt_A_01"!(ComponentPose=Local To Component(LocalPose=Use cached pose 'CRCache_FeatherDyn'()), bAlphaBoolEnabled:bool=True, Rotation="0,0,0") -> (Pose) {
    alphaInputType=Bool
    rotationMode=BMM_Additive
    translationMode=BMM_Additive
    alphaBoolBlend {blendOption=Linear}
    boneToModify {boneName=dyn_skirt_A_01}
    translation {x=0}
   }
   n13 = RigidBody!(ComponentPose=@n12, Alpha, OverrideWorldGravity:Vector="0,0,-980") -> (Pose) {
    bClampLinearTranslationLimitToRefPose=true
    bOverrideWorldGravity=true
    overridePhysicsAsset=&/M_MED_GourdRiddance_Physics
    simulationSpace=BaseBoneSpace
    alphaBoolBlend {blendOption=Linear}
    baseBoneRef {boneName=pelvis}
    componentAppliedLinearAccClamp=<3000 3000 3000>
    componentLinearAccScale=<0.2 0.1 0.1>
    componentLinearVelScale=<0.2 0.1 0.1>
    overrideWorldGravity {z=-980}
    simSpaceSettings {
     maxAngularAcceleration=1002
     maxAngularVelocity=3
     maxLinearAcceleration=700
     maxLinearVelocity=500
     velocityScaleZ=0
     worldAlpha=0.5
     externalLinearDragV=<0.2 0.2>
    }
   }
   n14 = Get bIsCrouching()
   n15 = "Transform (Modify) Bone - Bone: dyn_skirt_E_01"!(ComponentPose=@n13, bAlphaBoolEnabled:bool=@n14, Rotation="-1.2,-2.2,11") -> (Pose) {
    alphaInputType=Bool
    rotationMode=BMM_Additive
    translationMode=BMM_Additive
    alphaBoolBlend {blendOption=Linear}
    boneToModify {boneName=dyn_skirt_E_01}
    rotation=<rot -1.2 -2.2 11>
    translation=<-0.5218 2.0209 0.4065>
   }
   n16 = "Transform (Modify) Bone - Bone: dyn_skirt_E_02"!(ComponentPose=@n15, bAlphaBoolEnabled:bool=@n14, Rotation="-0.000,0.000,4.560") -> (Pose) {
    alphaInputType=Bool
    rotationMode=BMM_Additive
    translationMode=BMM_Additive
    alphaBoolBlend {blendOption=Linear}
    boneToModify {boneName=dyn_skirt_E_02}
    rotation {roll=4.56}
    translation=<-0.5218 0.0485 0.4065>
   }
   n17 = "Transform (Modify) Bone - Bone: dyn_skirt_D_02"!(ComponentPose=@n16, bAlphaBoolEnabled:bool=@n14, Rotation="0,0,5") -> (Pose) {
    alphaInputType=Bool
    rotationMode=BMM_Additive
    translationMode=BMM_Additive
    alphaBoolBlend {blendOption=Linear}
    boneToModify {boneName=dyn_skirt_D_02}
    rotation {roll=5}
    translation=<0.3691 -3.8841 -0.2557>
   }
   n18 = Get bIsDBNO()
   n19 = "Transform (Modify) Bone - Bone: dyn_skirt_E_01"(ComponentPose=@n17, bAlphaBoolEnabled=@n18, Rotation="24.291,-3.917,15.695") {
    alphaInputType=Bool
    rotationMode=BMM_Additive
    translationMode=BMM_Additive
    alphaBoolBlend {blendOption=Linear}
    boneToModify {boneName=dyn_skirt_E_01}
    rotation=<rot 24.291 -3.917 15.695>
    translation=<-0.5218 2.0209 -0.212>
   }
   n20 = "Transform (Modify) Bone - Bone: dyn_skirt_G_01"!(ComponentPose=@n19, bAlphaBoolEnabled:bool=@n18, Rotation="5.159,10.225,8.515") -> (Pose) {
    alphaInputType=Bool
    rotationMode=BMM_Additive
    translationMode=BMM_Additive
    alphaBoolBlend {blendOption=Linear}
    boneToModify {boneName=dyn_skirt_G_01}
    rotation=<rot 5.159 10.225 8.515>
    translation=<-0.5218 2.0209 0.4065>
   }
   n21 = "Transform (Modify) Bone - Bone: base_skirt_F"!(ComponentPose=@n20, bAlphaBoolEnabled:bool=@n18, Rotation="-1.851,8.851,12.011") -> (Pose) {
    alphaInputType=Bool
    rotationMode=BMM_Additive
    translationMode=BMM_Additive
    alphaBoolBlend {blendOption=Linear}
    boneToModify {boneName=base_skirt_F}
    rotation=<rot -1.851 8.851 12.011>
    translation=<-0.5218 2.0209 0.4065>
   }
   n22 = Get bIsInFrontEnd()
   n23 = "Transform (Modify) Bone - Bone: base_skirt_C"!(ComponentPose=@n21, bAlphaBoolEnabled:bool=@n22, Rotation="0,0,2") -> (Pose) {
    alphaInputType=Bool
    rotationMode=BMM_Additive
    translationMode=BMM_Additive
    alphaBoolBlend {blendOption=Linear}
    boneToModify {boneName=base_skirt_C}
    rotation {roll=2}
   }
   n24 = "Transform (Modify) Bone - Bone: base_skirt_E"!(ComponentPose=@n23, bAlphaBoolEnabled:bool=@n22, Rotation="3,0,3") -> (Pose) {
    alphaInputType=Bool
    rotationMode=BMM_Additive
    translationMode=BMM_Additive
    alphaBoolBlend {blendOption=Linear}
    boneToModify {boneName=base_skirt_E}
    rotation {pitch=3, roll=3}
   }
   n25 = "Transform (Modify) Bone - Bone: dyn_skirt_C_02"!(ComponentPose=@n24, bAlphaBoolEnabled:bool=@n22, Rotation="-5,0,0") -> (Pose) {
    alphaInputType=Bool
    rotationMode=BMM_Additive
    translationMode=BMM_Additive
    alphaBoolBlend {blendOption=Linear}
    boneToModify {boneName=dyn_skirt_C_02}
    rotation {pitch=-5}
   }
   n26 = "Transform (Modify) Bone - Bone: dyn_skirt_C_04"!(ComponentPose=@n25, bAlphaBoolEnabled:bool=True, Rotation="25,0,0") -> (Pose) {
    alphaInputType=Bool
    rotationMode=BMM_Additive
    translationMode=BMM_Additive
    alphaBoolBlend {blendOption=Linear}
    boneToModify {boneName=dyn_skirt_C_04}
    rotation {pitch=25}
   }
   n27 = "Transform (Modify) Bone - Bone: dyn_skirt_C_01"!(ComponentPose=@n26, bAlphaBoolEnabled:bool=True, Rotation="-2,0,0") -> (Pose) {
    alphaInputType=Bool
    rotationMode=BMM_Additive
    translationMode=BMM_Additive
    alphaBoolBlend {blendOption=Linear}
    boneToModify {boneName=dyn_skirt_C_01}
    rotation {pitch=-2}
   }
   Save cached pose 'CRCache_SkirtBase_Setup'(Pose=Component To Local(ComponentPose=@n27)) {cacheName=CRCache_SkirtBase_Setup}
   n28 = Anim Dynamics - Bone: dyn_vine01_l!(
    ComponentPose=Local To Component(LocalPose=Use cached pose 'CRCache_InputPose'())
    Alpha=1
    GravityOverride:Vector="0,0,-980"
    ExternalForce="0,0,0"
   ) -> (Pose) {
    angularDampingOverride=1
    angularSpringConstant=20000
    bAngularSpring=true
    bOverrideAngularDamping=true
    bUseGravityOverride=true
    gravityScale=0.35
    alphaBoolBlend {blendOption=Linear}
    boundBone {boneName=dyn_vine01_l}
    componentLinearAccScale=<1 1 1>
    componentLinearVelScale=<1 1 1>
    gravityOverride {z=-980}
    physicsBodyDefinitions [{
      boundBone {boneName=dyn_vine01_l}
      boxExtents=<20 30 30>
      constraintSetup {angularLimitsMax=<0 40 40>, angularLimitsMin=<0 -40 -40>, angularTarget=<1 0 0>}
      localJointOffset=<4 0 0>
     }]
   }
   n29 = Anim Dynamics - Bone: dyn_vine01_r!(ComponentPose=@n28, Alpha=1, GravityOverride:Vector="0,0,-980", ExternalForce="0,0,0") -> (Pose) {
    angularDampingOverride=1
    angularSpringConstant=20000
    bAngularSpring=true
    bOverrideAngularDamping=true
    bUseGravityOverride=true
    gravityScale=0.35
    alphaBoolBlend {blendOption=Linear}
    boundBone {boneName=dyn_vine01_r}
    componentLinearAccScale=<1 1 1>
    componentLinearVelScale=<1 1 1>
    gravityOverride {z=-980}
    physicsBodyDefinitions [{
      boundBone {boneName=dyn_vine01_r}
      boxExtents=<20 30 30>
      constraintSetup {angularLimitsMax=<30 40 40>, angularLimitsMin=<-30 -40 -40>, angularTarget=<1 0 0>}
      localJointOffset=<4 0 0>
     }]
   }
   n30 = Anim Dynamics - Bone: dyn_vine02_l!(ComponentPose=@n29, Alpha=1, GravityOverride:Vector="0,0,-980", ExternalForce="0,0,0") -> (Pose) {
    angularDampingOverride=1
    angularSpringConstant=20000
    bAngularSpring=true
    bOverrideAngularDamping=true
    bUseGravityOverride=true
    gravityScale=0.35
    alphaBoolBlend {blendOption=Linear}
    boundBone {boneName=dyn_vine02_l}
    componentLinearAccScale=<1 1 1>
    componentLinearVelScale=<1 1 1>
    gravityOverride {z=-980}
    physicsBodyDefinitions [{
      boundBone {boneName=dyn_vine02_l}
      boxExtents=<20 30 30>
      constraintSetup {angularLimitsMax=<0 40 40>, angularLimitsMin=<0 -40 -40>, angularTarget=<1 0 0>}
      localJointOffset=<4 0 0>
     }]
   }
   n31 = Anim Dynamics - Bone: dyn_vine02_r!(ComponentPose=@n30, Alpha=1, GravityOverride:Vector="0,0,-980", ExternalForce="0,0,0") -> (Pose) {
    angularDampingOverride=1
    angularSpringConstant=20000
    bAngularSpring=true
    bOverrideAngularDamping=true
    bUseGravityOverride=true
    gravityScale=0.35
    alphaBoolBlend {blendOption=Linear}
    boundBone {boneName=dyn_vine02_r}
    componentLinearAccScale=<1 1 1>
    componentLinearVelScale=<1 1 1>
    gravityOverride {z=-980}
    physicsBodyDefinitions [{
      boundBone {boneName=dyn_vine02_r}
      boxExtents=<20 30 30>
      constraintSetup {angularLimitsMax=<0 40 40>, angularLimitsMin=<0 -40 -40>, angularTarget=<1 0 0>}
      localJointOffset=<4 0 0>
     }]
   }
   n32 = Anim Dynamics - Bone: dyn_vine03_l!(ComponentPose=@n31, Alpha=1, GravityOverride:Vector="0,0,-980", ExternalForce="0,0,0") -> (Pose) {
    angularDampingOverride=1
    angularSpringConstant=20000
    bAngularSpring=true
    bOverrideAngularDamping=true
    bUseGravityOverride=true
    gravityScale=0.35
    alphaBoolBlend {blendOption=Linear}
    boundBone {boneName=dyn_vine03_l}
    componentLinearAccScale=<1 1 1>
    componentLinearVelScale=<1 1 1>
    gravityOverride {z=-980}
    physicsBodyDefinitions [{
      boundBone {boneName=dyn_vine03_l}
      boxExtents=<20 30 30>
      constraintSetup {angularLimitsMax=<0 40 40>, angularLimitsMin=<0 -40 -40>, angularTarget=<1 0 0>}
      localJointOffset=<4 0 0>
     }]
   }
   n33 = Anim Dynamics - Bone: dyn_shoulderPDM02_l(ComponentPose=@n32, Alpha=1, ExternalForce="0,0,0") {
    angularDampingOverride=1
    angularSpringConstant=20000
    bAngularSpring=true
    bOverrideAngularDamping=true
    bUseGravityOverride=true
    gravityScale=0.35
    simulationSpace=RootRelative
    alphaBoolBlend {blendOption=Linear}
    boundBone {boneName=dyn_shoulderPDM02_l}
    componentLinearAccScale=<1 1 1>
    componentLinearVelScale=<1 1 1>
    gravityOverride {z=-980}
    physicsBodyDefinitions [{
      boundBone {boneName=dyn_shoulderPDM02_l}
      boxExtents=<20 30 30>
      constraintSetup {angularLimitsMax=<15 15 15>, angularLimitsMin=<-15 -15 -15>, angularTarget=<1 0 0>}
      localJointOffset=<4 0 0>
     }]
   }
   n34 = Anim Dynamics - Bone: dyn_pant01_l!(ComponentPose=@n33, Alpha=1, GravityOverride:Vector="0,0,-980", ExternalForce="0,0,0") -> (Pose) {
    angularDampingOverride=1
    angularSpringConstant=20000
    bAngularSpring=true
    bOverrideAngularDamping=true
    bUseGravityOverride=true
    gravityScale=0.35
    alphaBoolBlend {blendOption=Linear}
    boundBone {boneName=dyn_pant01_l}
    componentLinearAccScale=<1 1 1>
    componentLinearVelScale=<1 1 1>
    gravityOverride {z=-980}
    physicsBodyDefinitions [{
      boundBone {boneName=dyn_pant01_l}
      boxExtents=<20 30 30>
      constraintSetup {angularLimitsMax=<0 3 60>, angularLimitsMin=<0 -15 -30>, angularTarget=<1 0 0>}
      localJointOffset=<4 0 0>
     }]
   }
   n35 = Anim Dynamics - Bone: dyn_pant01_r!(ComponentPose=@n34, Alpha=1, GravityOverride:Vector="0,0,-980", ExternalForce="0,0,0") -> (Pose) {
    angularDampingOverride=1
    angularSpringConstant=20000
    bAngularSpring=true
    bOverrideAngularDamping=true
    bUseGravityOverride=true
    gravityScale=0.35
    alphaBoolBlend {blendOption=Linear}
    boundBone {boneName=dyn_pant01_r}
    componentLinearAccScale=<1 1 1>
    componentLinearVelScale=<1 1 1>
    gravityOverride {z=-980}
    physicsBodyDefinitions [{
      boundBone {boneName=dyn_pant01_r}
      boxExtents=<20 30 30>
      constraintSetup {angularLimitsMax=<0 3 30>, angularLimitsMin=<0 -15 -60>, angularTarget=<1 0 0>}
      localJointOffset=<4 0 0>
     }]
   }
   n36 = Anim Dynamics - Bone: dyn_pant02_l!(ComponentPose=@n35, Alpha=1, GravityOverride:Vector="0,0,-980", ExternalForce="0,0,0") -> (Pose) {
    angularDampingOverride=1
    angularSpringConstant=20000
    bAngularSpring=true
    bOverrideAngularDamping=true
    bUseGravityOverride=true
    gravityScale=0.35
    alphaBoolBlend {blendOption=Linear}
    boundBone {boneName=dyn_pant02_l}
    componentLinearAccScale=<1 1 1>
    componentLinearVelScale=<1 1 1>
    gravityOverride {z=-980}
    physicsBodyDefinitions [{
      boundBone {boneName=dyn_pant02_l}
      boxExtents=<20 30 30>
      constraintSetup {angularLimitsMax=<0 3 30>, angularLimitsMin=<0 -15 -30>, angularTarget=<1 0 0>}
      localJointOffset=<4 0 0>
     }]
   }
   n37 = Anim Dynamics - Bone: dyn_pant02_r!(ComponentPose=@n36, Alpha=1, GravityOverride:Vector="0,0,-980", ExternalForce="0,0,0") -> (Pose) {
    angularDampingOverride=1
    angularSpringConstant=20000
    bAngularSpring=true
    bOverrideAngularDamping=true
    bUseGravityOverride=true
    gravityScale=0.35
    alphaBoolBlend {blendOption=Linear}
    boundBone {boneName=dyn_pant02_r}
    componentLinearAccScale=<1 1 1>
    componentLinearVelScale=<1 1 1>
    gravityOverride {z=-980}
    physicsBodyDefinitions [{
      boundBone {boneName=dyn_pant02_r}
      boxExtents=<20 30 30>
      constraintSetup {angularLimitsMax=<0 3 30>, angularLimitsMin=<0 -15 -30>, angularTarget=<1 0 0>}
      localJointOffset=<4 0 0>
     }]
   }
   Save cached pose 'CRCache_AnimDyn'(Pose=Component To Local(ComponentPose=@n37)) {cacheName=CRCache_AnimDyn}
   n38 = RigidBody!(ComponentPose=Local To Component(LocalPose=Use cached pose 'CRCache_AnimDyn'()), Alpha, OverrideWorldGravity="0,0,0", ExternalForce:Vector) -> (Pose) {
    bOverrideWorldGravity=true
    overridePhysicsAsset=&/M_MED_GourdRiddance_Feathers_Physics
    simulationSpace=BaseBoneSpace
    alphaBoolBlend {blendOption=Linear}
    baseBoneRef {boneName=pelvis}
    componentAppliedLinearAccClamp=<500 500 500>
    componentLinearAccScale=<0.05 0.05 0.05>
    componentLinearVelScale=<0.05 0.05 0.05>
    simSpaceSettings {
     maxAngularAcceleration=500
     maxAngularVelocity=3
     maxLinearAcceleration=300
     maxLinearVelocity=500
     velocityScaleZ=0
     worldAlpha=0.2
     externalLinearDragV=<0.3 0.3>
    }
   }
   n39 = Anim Dynamics - Bone: dyn_tipScarf_bck!(ComponentPose=@n38, bAlphaBoolEnabled:bool=True, GravityOverride:Vector="0,0,-980", ExternalForce="0,0,0") -> (Pose) {
    alphaInputType=Bool
    angularDampingOverride=1
    angularSpringConstant=20000
    bAngularSpring=true
    bOverrideAngularDamping=true
    bUseGravityOverride=true
    gravityScale=0.35
    alphaBoolBlend {blendInTime=0.2, blendOption=Linear, blendOutTime=0.2}
    boundBone {boneName=dyn_tipScarf_bck}
    componentLinearAccScale=<1 1 1>
    componentLinearVelScale=<1 1 1>
    gravityOverride {z=-980}
    physicsBodyDefinitions [{
      boundBone {boneName=dyn_tipScarf_bck}
      boxExtents=<20 30 30>
      constraintSetup {angularLimitsMax=<0 1 20>, angularLimitsMin=<0 -80 -20>, angularTarget=<1 0 0>}
      localJointOffset=<8 0 0>
     }]
   }
   Save cached pose 'CRCache_FeatherDyn'(Pose=Component To Local(ComponentPose=@n39)) {cacheName=CRCache_FeatherDyn}
   n40 = RigidBody!(ComponentPose=Local To Component(LocalPose=Use cached pose 'CRCache_FeatherDyn'()), Alpha, OverrideWorldGravity="0,0,-980") -> (Pose) {
    bClampLinearTranslationLimitToRefPose=true
    bOverrideWorldGravity=true
    overridePhysicsAsset=&/M_MED_GourdRiddance_EmotePlayin_Physics
    simulationSpace=BaseBoneSpace
    alphaBoolBlend {blendOption=Linear}
    baseBoneRef {boneName=pelvis}
    componentAppliedLinearAccClamp=<3000 3000 3000>
    componentLinearAccScale=<0.2 0.1 0.1>
    componentLinearVelScale=<0.2 0.1 0.1>
    overrideWorldGravity {z=-980}
    simSpaceSettings {
     maxAngularAcceleration=1002
     maxAngularVelocity=3
     maxLinearAcceleration=700
     maxLinearVelocity=500
     velocityScaleZ=0
     worldAlpha=0.5
     externalLinearDragV=<0.2 0.2>
    }
   }
   n41 = "Transform (Modify) Bone - Bone: dyn_outbelt_bck_l"(ComponentPose=@n40, Alpha=1, Translation="-22.320,2.061,-2.129", Rotation="-0.576,-3.541,0.585", Scale="1,1,1") {
    rotationMode=BMM_Additive
    rotationSpace=BCS_BoneSpace
    translationMode=BMM_Additive
    translationSpace=BCS_BoneSpace
    alphaBoolBlend {blendOption=Linear}
    boneToModify {boneName=dyn_outbelt_bck_l}
    rotation=<rot -0.576 -3.541 0.585>
    translation=<-22.32 2.061 -2.129>
   }
   n42 = "Transform (Modify) Bone - Bone: dyn_outbelt_r"(ComponentPose=@n41, Alpha=1, Translation="-2.090,-1.181,0.037", Rotation="-3.674,41.789,4.214", Scale="1,1,1") {
    rotationMode=BMM_Additive
    rotationSpace=BCS_BoneSpace
    translationMode=BMM_Additive
    translationSpace=BCS_BoneSpace
    alphaBoolBlend {blendOption=Linear}
    boneToModify {boneName=dyn_outbelt_r}
    rotation=<rot -3.674 41.789 4.214>
    translation=<-2.0904 -1.1813 0.037>
   }
   n43 = "Constraint - base_feather_a01_r (PPPP)"!(
    ComponentPose=@n42
    bAlphaBoolEnabled:bool=True
    ConstraintWeights_0=0.5
    ConstraintWeights_1:float=0.2
    ConstraintWeights_2:float=0.2
    ConstraintWeights_3:float=0.1
   ) -> (Pose) {
    alphaInputType=Bool
    alphaBoolBlend {blendInTime=5, blendOption=CubicInOut, blendOutTime=5}
    boneToModify {boneName=base_feather_a01_r}
    constraintSetup [
     {perAxis {bX=false, bZ=false}, targetBone {boneName=dyn_skirt_G_01}}
     {perAxis {bX=false, bZ=false}, targetBone {boneName=dyn_feather_b02_r}}
     {perAxis {bX=false, bZ=false}, targetBone {boneName=thigh_r}}
     {perAxis {bX=false, bZ=false}, targetBone {boneName=deform_groin_r}}
    ]
    constraintWeights [1,1,1,1]
   }
   n44 = "Constraint - base_feather_b01_r (PPPP)"!(
    ComponentPose=@n43
    bAlphaBoolEnabled:bool=True
    ConstraintWeights_0=0.4
    ConstraintWeights_1:float=0.3
    ConstraintWeights_2:float=0.2
    ConstraintWeights_3:float=0.1
   ) -> (Pose) {
    alphaInputType=Bool
    alphaBoolBlend {blendInTime=5, blendOption=CubicInOut, blendOutTime=5}
    boneToModify {boneName=base_feather_b01_r}
    constraintSetup [
     {perAxis {bX=false, bZ=false}, targetBone {boneName=dyn_skirt_F_01}}
     {perAxis {bX=false, bZ=false}, targetBone {boneName=dyn_feather_c01_r}}
     {perAxis {bX=false, bZ=false}, targetBone {boneName=thigh_r}}
     {perAxis {bX=false, bZ=false}, targetBone {boneName=deform_groin_r}}
    ]
    constraintWeights [1,1,1,1]
   }
   Save cached pose 'CRCache_Emote'(Pose=Component To Local(ComponentPose=@n44)) {cacheName=CRCache_Emote}
   n45 = Look At - Bone: dyn_skirt_A_01(ComponentPose=Local To Component(LocalPose=Use cached pose 'CRCache_FeatherDyn'()), Alpha=1) {
    alphaBoolBlend {blendOption=Linear}
    boneToModify {boneName=dyn_skirt_A_01}
    lookAtLocation=<57.6574 14.9038 11.4647>
    lookAtTarget {boneReference {boneName=root}}
    lookAt_Axis {axis=<1 0>}
   }
   n46 = Look At - Bone: base_skirt_B(ComponentPose=@n45, Alpha=1) {
    alphaBoolBlend {blendOption=Linear}
    boneToModify {boneName=base_skirt_B}
    lookAtLocation=<63.7612 -10.1259 9.0182>
    lookAtTarget {boneReference {boneName=root}}
    lookAt_Axis {axis=<1 0>}
   }
   n47 = Look At - Bone: base_skirt_F(ComponentPose=@n46, Alpha=1) {
    alphaBoolBlend {blendOption=Linear}
    boneToModify {boneName=base_skirt_F}
    lookAtLocation=<-52.0591 -14.1608 -13.1348>
    lookAtTarget {boneReference {boneName=root}}
    lookAt_Axis {axis=<1 0>}
   }
   n48 = Look At - Bone: base_skirt_G(ComponentPose=@n47, Alpha=1) {
    alphaBoolBlend {blendOption=Linear}
    boneToModify {boneName=base_skirt_G}
    lookAtLocation=<-43.6433 16.3582 7.7209>
    lookAtTarget {boneReference {boneName=root}}
    lookAt_Axis {axis=<1 0>}
   }
   n49 = "Transform (Modify) Bone - Bone: base_skirt_E"!(ComponentPose=@n48, bAlphaBoolEnabled=True, Translation:Vector="1.700,-3.687,1.332") -> (Pose) {
    alphaInputType=Bool
    rotationMode=BMM_Additive
    translationMode=BMM_Additive
    alphaBoolBlend {blendOption=Linear}
    boneToModify {boneName=base_skirt_E}
    rotation=<rot -0.601 5.7024 70.5625>
    translation=<1.7 -3.687 1.3317>
   }
   n50 = Look At - Bone: base_skirt_D(ComponentPose=@n49, Alpha=1) {
    alphaBoolBlend {blendOption=Linear}
    boneToModify {boneName=base_skirt_D}
    lookAtLocation=<8.4391 -62.1745 -0.5312>
    lookAtTarget {boneReference {boneName=root}}
    lookAt_Axis {axis=<1 0>}
   }
   n51 = Look At - Bone: base_skirt_C(ComponentPose=@n50, Alpha=1) {
    alphaBoolBlend {blendOption=Linear}
    boneToModify {boneName=base_skirt_C}
    lookAtLocation=<32.3714 -49.8301 -1.868>
    lookAtTarget {boneReference {boneName=root}}
    lookAt_Axis {axis=<1 0>}
   }
   n52 = "Transform (Modify) Bone - Bone: dyn_feather_d01_r"!(ComponentPose=@n51, bAlphaBoolEnabled:bool=True) -> (Pose) {
    alphaInputType=Bool
    rotationMode=BMM_Additive
    rotationSpace=BCS_BoneSpace
    translationMode=BMM_Additive
    translationSpace=BCS_BoneSpace
    alphaBoolBlend {blendOption=Linear}
    boneToModify {boneName=dyn_feather_d01_r}
    rotation=<rot 36.2722 0.0195 0.0343>
   }
   n53 = "Transform (Modify) Bone - Bone: dyn_feather_c01_r"!(ComponentPose=@n52, bAlphaBoolEnabled:bool=True) -> (Pose) {
    alphaInputType=Bool
    rotationMode=BMM_Additive
    rotationSpace=BCS_BoneSpace
    translationMode=BMM_Additive
    translationSpace=BCS_BoneSpace
    alphaBoolBlend {blendOption=Linear}
    boneToModify {boneName=dyn_feather_c01_r}
    rotation=<rot 40.6016 -0.0002 10.642>
   }
   n54 = "Transform (Modify) Bone - Bone: dyn_feather_b01_r"!(ComponentPose=@n53, bAlphaBoolEnabled:bool=True) -> (Pose) {
    alphaInputType=Bool
    rotationMode=BMM_Additive
    rotationSpace=BCS_BoneSpace
    translationMode=BMM_Additive
    translationSpace=BCS_BoneSpace
    alphaBoolBlend {blendOption=Linear}
    boneToModify {boneName=dyn_feather_b01_r}
    rotation=<rot 34.6512 -0.0015 0.0035>
   }
   n55 = "Transform (Modify) Bone - Bone: dyn_feather_a01_r"!(ComponentPose=@n54, bAlphaBoolEnabled:bool=True) -> (Pose) {
    alphaInputType=Bool
    rotationMode=BMM_Additive
    rotationSpace=BCS_BoneSpace
    translationMode=BMM_Additive
    translationSpace=BCS_BoneSpace
    alphaBoolBlend {blendOption=Linear}
    boneToModify {boneName=dyn_feather_a01_r}
    rotation=<rot 40.1188 39.7275 14.4757>
   }
   Save cached pose 'CRCache_SlidingDisable'(Pose=Component To Local(ComponentPose=@n55)) {cacheName=CRCache_SlidingDisable}
   n56 = Look At - Bone: dyn_skirt_A_01(ComponentPose=Local To Component(LocalPose=Use cached pose 'CRCache_FeatherDyn'()), Alpha=1) {
    alphaBoolBlend {blendOption=Linear}
    boneToModify {boneName=dyn_skirt_A_01}
    lookAtLocation=<57.6574 42.696 19.2809>
    lookAtTarget {boneReference {boneName=root}}
    lookAt_Axis {axis=<1 0>}
   }
   n57 = Look At - Bone: base_skirt_B(ComponentPose=@n56, Alpha=1) {
    alphaBoolBlend {blendOption=Linear}
    boneToModify {boneName=base_skirt_B}
    lookAtLocation=<75.7043 20.7542 20.0602>
    lookAtTarget {boneReference {boneName=root}}
    lookAt_Axis {axis=<1 0>}
   }
   n58 = Look At - Bone: base_skirt_F(ComponentPose=@n57, Alpha=1) {
    alphaBoolBlend {blendOption=Linear}
    boneToModify {boneName=base_skirt_F}
    lookAtLocation=<-52.0591 -13.744 -5.9438>
    lookAtTarget {boneReference {boneName=root}}
    lookAt_Axis {axis=<1 0>}
   }
   n59 = Look At - Bone: base_skirt_G(ComponentPose=@n58, Alpha=1) {
    alphaBoolBlend {blendOption=Linear}
    boneToModify {boneName=base_skirt_G}
    lookAtLocation=<-43.6433 46.7793 9.155>
    lookAtTarget {boneReference {boneName=root}}
    lookAt_Axis {axis=<1 0>}
   }
   n60 = "Transform (Modify) Bone - Bone: base_skirt_E"(ComponentPose=@n59, bAlphaBoolEnabled=True, Translation="1.700,-3.687,1.332") {
    alphaInputType=Bool
    rotationMode=BMM_Additive
    translationMode=BMM_Additive
    alphaBoolBlend {blendOption=Linear}
    boneToModify {boneName=base_skirt_E}
    rotation=<rot -0.601 5.7024 70.5625>
    translation=<1.7 -3.687 1.3317>
   }
   n61 = Look At - Bone: base_skirt_D(ComponentPose=@n60, Alpha=1) {
    alphaBoolBlend {blendOption=Linear}
    boneToModify {boneName=base_skirt_D}
    lookAtLocation=<8.4391 -62.1745 -0.5312>
    lookAtTarget {boneReference {boneName=root}}
    lookAt_Axis {axis=<1 0>}
   }
   n62 = Look At - Bone: base_skirt_C(ComponentPose=@n61, Alpha=1) {
    alphaBoolBlend {blendOption=Linear}
    boneToModify {boneName=base_skirt_C}
    lookAtLocation=<32.3714 -49.8301 -1.868>
    lookAtTarget {boneReference {boneName=root}}
    lookAt_Axis {axis=<1 0>}
   }
   n63 = RigidBody(ComponentPose=@n62, OverrideWorldGravity="0,0,-980") {
    bClampLinearTranslationLimitToRefPose=true
    bOverrideWorldGravity=true
    overridePhysicsAsset=&/M_MED_GourdRiddance_Sliding_Physics
    simulationSpace=BaseBoneSpace
    alphaBoolBlend {blendOption=Linear}
    baseBoneRef {boneName=pelvis}
    componentAppliedLinearAccClamp=<3000 3000 3000>
    componentLinearAccScale=<0.3 0.2 0.2>
    componentLinearVelScale=<0.3 0.2 0.2>
    overrideWorldGravity {z=-980}
    simSpaceSettings {
     maxAngularAcceleration=1002
     maxAngularVelocity=3
     maxLinearAcceleration=700
     maxLinearVelocity=500
     velocityScaleZ=0
     worldAlpha=0.5
     externalLinearDragV=<0.2 0.2>
    }
   }
   n64 = "Transform (Modify) Bone - Bone: dyn_feather_d01_r"(ComponentPose=@n63, bAlphaBoolEnabled=True) {
    alphaInputType=Bool
    rotationMode=BMM_Additive
    rotationSpace=BCS_BoneSpace
    translationMode=BMM_Additive
    translationSpace=BCS_BoneSpace
    alphaBoolBlend {blendOption=Linear}
    boneToModify {boneName=dyn_feather_d01_r}
    rotation=<rot 36.2722 0.0195 0.0343>
   }
   n65 = "Transform (Modify) Bone - Bone: dyn_feather_c01_r"(ComponentPose=@n64, bAlphaBoolEnabled=True) {
    alphaInputType=Bool
    rotationMode=BMM_Additive
    rotationSpace=BCS_BoneSpace
    translationMode=BMM_Additive
    translationSpace=BCS_BoneSpace
    alphaBoolBlend {blendOption=Linear}
    boneToModify {boneName=dyn_feather_c01_r}
    rotation=<rot 40.6016 -0.0002 10.642>
   }
   n66 = "Transform (Modify) Bone - Bone: dyn_feather_b01_r"(ComponentPose=@n65, bAlphaBoolEnabled=True) {
    alphaInputType=Bool
    rotationMode=BMM_Additive
    rotationSpace=BCS_BoneSpace
    translationMode=BMM_Additive
    translationSpace=BCS_BoneSpace
    alphaBoolBlend {blendOption=Linear}
    boneToModify {boneName=dyn_feather_b01_r}
    rotation=<rot 34.6512 -0.0015 0.0035>
   }
   n67 = "Transform (Modify) Bone - Bone: dyn_feather_a01_r"(ComponentPose=@n66, bAlphaBoolEnabled=True) {
    alphaInputType=Bool
    rotationMode=BMM_Additive
    rotationSpace=BCS_BoneSpace
    translationMode=BMM_Additive
    translationSpace=BCS_BoneSpace
    alphaBoolBlend {blendOption=Linear}
    boneToModify {boneName=dyn_feather_a01_r}
    rotation=<rot 40.1188 39.7275 14.4757>
   }
   Save cached pose 'CRCache_Sliding'(Pose=Component To Local(ComponentPose=@n67)) {cacheName=CRCache_Sliding}
   n68 = Look At - Bone: dyn_skirt_A_01(ComponentPose=Local To Component(LocalPose=Use cached pose 'CRCache_FeatherDyn'()), Alpha=1) {
    alphaBoolBlend {blendOption=Linear}
    boneToModify {boneName=dyn_skirt_A_01}
    lookAtLocation=<56.4343 85.4028 19.2809>
    lookAtTarget {boneReference {boneName=root}}
    lookAt_Axis {axis=<1 0>}
   }
   n69 = Look At - Bone: base_skirt_B(ComponentPose=@n68, Alpha=1) {
    alphaBoolBlend {blendOption=Linear}
    boneToModify {boneName=base_skirt_B}
    lookAtLocation=<89.5941 38.0401 8.9414>
    lookAtTarget {boneReference {boneName=root}}
    lookAt_Axis {axis=<1 0>}
   }
   n70 = "Transform (Modify) Bone - Bone: base_skirt_F"!(ComponentPose=@n69, bAlphaBoolEnabled=True, Translation:Vector="-0.000,0.000,-2.542") -> (Pose) {
    alphaInputType=Bool
    rotationMode=BMM_Additive
    translationMode=BMM_Additive
    alphaBoolBlend {blendOption=Linear}
    boneToModify {boneName=base_skirt_F}
    rotation=<rot 0.8971 -6.3383 117.3225>
    translation {z=-2.542}
   }
   n71 = "Transform (Modify) Bone - Bone: base_skirt_G"!(ComponentPose=@n70, bAlphaBoolEnabled:bool=True, Translation="0.000,0.000,-6.364") -> (Pose) {
    alphaInputType=Bool
    rotationMode=BMM_Additive
    translationMode=BMM_Additive
    alphaBoolBlend {blendOption=Linear}
    boneToModify {boneName=base_skirt_G}
    rotation=<rot 3.6053 -34.1918 125.11>
    translation {z=-6.364}
   }
   n72 = "Transform (Modify) Bone - Bone: base_skirt_E"(ComponentPose=@n71, bAlphaBoolEnabled=True, Translation="1.700,-3.687,-1.553") {
    alphaInputType=Bool
    rotationMode=BMM_Additive
    translationMode=BMM_Additive
    alphaBoolBlend {blendOption=Linear}
    boneToModify {boneName=base_skirt_E}
    rotation=<rot 18.2618 -7.2569 94.6555>
    translation=<1.7 -3.687 -1.553>
   }
   n73 = Look At - Bone: base_skirt_D(ComponentPose=@n72, Alpha=1) {
    alphaBoolBlend {blendOption=Linear}
    boneToModify {boneName=base_skirt_D}
    lookAtLocation=<39.7075 -51.462 -23.0742>
    lookAtTarget {boneReference {boneName=root}}
    lookAt_Axis {axis=<1 0>}
   }
   n74 = Look At - Bone: base_skirt_C(ComponentPose=@n73, Alpha=1) {
    alphaBoolBlend {blendOption=Linear}
    boneToModify {boneName=base_skirt_C}
    lookAtLocation=<65.1184 -21.4065 -1.868>
    lookAtTarget {boneReference {boneName=root}}
    lookAt_Axis {axis=<1 0>}
   }
   n75 = RigidBody(ComponentPose=@n74, OverrideWorldGravity="0,0,-980") {
    bClampLinearTranslationLimitToRefPose=true
    bOverrideWorldGravity=true
    overridePhysicsAsset=&M_MED_GourdRiddance_Sliding_Physics
    simulationSpace=BaseBoneSpace
    alphaBoolBlend {blendOption=Linear}
    baseBoneRef {boneName=pelvis}
    componentAppliedLinearAccClamp=<3000 3000 3000>
    componentLinearAccScale=<0.3 0.2 0.2>
    componentLinearVelScale=<0.3 0.2 0.2>
    overrideWorldGravity {z=-980}
    simSpaceSettings {
     maxAngularAcceleration=1002
     maxAngularVelocity=3
     maxLinearAcceleration=700
     maxLinearVelocity=500
     velocityScaleZ=0
     worldAlpha=0.5
     externalLinearDragV=<0.2 0.2>
    }
   }
   n76 = "Transform (Modify) Bone - Bone: dyn_feather_d01_r"(ComponentPose=@n75, bAlphaBoolEnabled=True) {
    alphaInputType=Bool
    rotationMode=BMM_Additive
    rotationSpace=BCS_BoneSpace
    translationMode=BMM_Additive
    translationSpace=BCS_BoneSpace
    alphaBoolBlend {blendOption=Linear}
    boneToModify {boneName=dyn_feather_d01_r}
    rotation=<rot 39.7053 -21.1641 -11.735>
   }
   n77 = "Transform (Modify) Bone - Bone: dyn_feather_c01_r"(ComponentPose=@n76, bAlphaBoolEnabled=True) {
    alphaInputType=Bool
    rotationMode=BMM_Additive
    rotationSpace=BCS_BoneSpace
    translationMode=BMM_Additive
    translationSpace=BCS_BoneSpace
    alphaBoolBlend {blendOption=Linear}
    boneToModify {boneName=dyn_feather_c01_r}
    rotation=<rot 41.6354 16.7044 11.583>
   }
   n78 = "Transform (Modify) Bone - Bone: dyn_feather_b01_r"(ComponentPose=@n77, bAlphaBoolEnabled=True) {
    alphaInputType=Bool
    rotationMode=BMM_Additive
    rotationSpace=BCS_BoneSpace
    translationMode=BMM_Additive
    translationSpace=BCS_BoneSpace
    alphaBoolBlend {blendOption=Linear}
    boneToModify {boneName=dyn_feather_b01_r}
    rotation=<rot 22.811 -26.6534 -33.7856>
   }
   n79 = "Transform (Modify) Bone - Bone: dyn_feather_a01_r"(ComponentPose=@n78, bAlphaBoolEnabled=True) {
    alphaInputType=Bool
    rotationMode=BMM_Additive
    rotationSpace=BCS_BoneSpace
    translationMode=BMM_Additive
    translationSpace=BCS_BoneSpace
    alphaBoolBlend {blendOption=Linear}
    boneToModify {boneName=dyn_feather_a01_r}
    rotation=<rot 12.2712 31.1519 11.8972>
   }
   Save cached pose 'CRCache_BaseBallSliding'(Pose=Component To Local(ComponentPose=@n79)) {cacheName=CRCache_BaseBallSliding}
  ]}
 EventGraph {nodes [
   n1 = Event BlueprintUpdateAnimation()
   > n2 = Sequence!(execute) -> (then_0, then_1, then_2:Exec, then_3:Exec, then_4:Exec, then_5:Exec)
   n3 = GetAnimInstance(self=GetOwnerSkeletalMeshComponent())
   n4 = Branch(
    execute=@n2.then_4
    Condition=AND Boolean(
     A=Get bIsInFrontEnd()
     B=Montage_IsPlaying(self=@n3, Montage=/BRCosmetics/Animation/Game/MainPlayer/Emotes/BalletSpin/Emote_BalletSpin_CMM_M)
    )
   )
   > Set BalletSpin_FE(BalletSpin_FE:bool=true) -> (Output_Get:bool)
   Set BalletSpin_FE(execute=@n4.else, BalletSpin_FE=false)
   n5 = Gravity Override State Function DEPRECATED(
    execute=@n2.then_0
    CharacterPartAnimInstance=Self-Reference()
    GameTimeInSeconds=GetGameTimeInSeconds()
    GravityOverrideParams=Get Gravity Override Params()
   )
   > Set GravityOverride(GravityOverride=@n5)
   n6 = GetOwnerSkeletalMeshComponent()
   n7 = BreakRotator(InRot=BreakTransform(InTransform=GetDeltaTransformFromRefPose(self=@n6, BoneName=head, BaseName=neck_02)).Rotation)
   n8 = Get Y percentage() -> (Y percentage:float)
   n9 = Get X percentage() -> (X percentage:float)
   n10 = Get Z percentage() -> (Z percentage:float)
   > Set head rot(head rot:Rotator=MakeRotator(
     Roll=float * float!(A=@n7, B=-1, C:double=@n9) -> (ReturnValue)
     Pitch=float * float(A=@n7.Pitch, B=-1, C=@n8)
     Yaw=float * float(A=@n7.Yaw, B=-1, C=@n10)
    )) -> (Output_Get:Rotator)
   n11 = Cast To FortPlayerAnimInstance(execute=@n2.then_2, Object=GetAnimInstance(self=GetOwnerSkeletalMeshComponent()))
   > Set MeeleWeapon(MeeleWeapon:float=1) -> (Output_Get:float)
   > n12 = Switch on EFortWeaponCoreAnimation!(execute, Selection:EFortWeaponCoreAnimation Enum=Get WeaponCoreAnim(self=@n11)) -> (
    Melee:Exec
    Pistol:Exec
    Shotgun:Exec
    PaperBlueprint:Exec
    Rifle:Exec
    MeleeOneHand:Exec
    MachinePistol:Exec
    RocketLauncher:Exec
    GrenadeLauncher:Exec
    GoingCommando:Exec
    AssaultRifle:Exec
    TacticalShotgun:Exec
    SniperRifle:Exec
    TrapPlacement:Exec
    ShoulderLauncher:Exec
    AbilityDecoTool:Exec
    Crossbow:Exec
    C4:Exec
    RemoteControl:Exec
    DualWield:Exec
    AR_BullPup:Exec
    AR_ForwardGrip:Exec
    MedPackPaddles:Exec
    SMG_P90:Exec
    AR_DrumGun:Exec
    Consumable_Small:Exec
    Consumable_Large:Exec
    Balloon:Exec
    MountedTurret:Exec
    CreativeTool:Exec
    ExplosiveBow:Exec
    AshtonIndigo:Exec
    AshtonChicago:Exec
    MeleeDualWield:Exec
    Unarmed:Exec
   )
   Branch(execute=[@n12.Melee, @n12.MeleeOneHand])
   > Set MeeleWeapon(MeeleWeapon=0)
   Set SkirtRotation(
    execute=@n2.then_3
    SkirtRotation:double=GetAlphaFromReferenceTransformDelta(SkeletalMeshComponent=Get Skel Mesh(), BoneName=pelvis, BaseName=root, InputRange=(X=10,Y=-50), TwistAxis="0,0,1")
   ) -> (Output_Get:double)
   n13 = Get bIsSkyDiving()
   Set floatingVar(
    execute=@n2.then_1
    floatingVar:double=FInterpTo(Current=Get floatingVar(), Target="To Float (Boolean)"(InBool=@n13), DeltaTime=@n1.DeltaTimeX, InterpSpeed=2)
   ) -> (Output_Get:double)
   n14 = Get bIsMoving2D()
   > Set Moving2d(Moving2d:double=FInterpTo(Current=Get Moving2d(), Target="To Float (Boolean)"(InBool=@n14), DeltaTime=@n1.DeltaTimeX, InterpSpeed=2)) -> (Output_Get:double)
   n15 = Get bIsFalling()
   n16 = Get bIsJumping()
   n17 = Get bIsCrouching()
   n18 = AND Boolean(A=Get bIsSprinting(), B=NOT Boolean(A=OR Boolean!(A=@n15, B=@n16, C:bool=@n17) -> (ReturnValue)))
   > Set IsRunning(IsRunning:bool=@n18) -> (Output_Get:bool)
   > Set isRunFloat(isRunFloat:double=FInterpTo(Current=Get isRunFloat(), Target="To Float (Boolean)"(InBool=@n18), DeltaTime=@n1.DeltaTimeX, InterpSpeed=2.5)) -> (Output_Get:double)
   n19 = Get bIsSliding()
   > Set isSliding(isSliding:double=FInterpTo(Current=Get isSliding(), Target="To Float (Boolean)"(InBool=@n19), DeltaTime=@n1.DeltaTimeX, InterpSpeed=2)) -> (Output_Get:double)
   n20 = Get bRigidBodyEnabled()
   n21 = Get bIsKneeSliding()
   > Set SlidingRBdisabled(SlidingRBdisabled:bool=AND Boolean(A=@n21, B=NOT Boolean(A=@n20))) -> (Output_Get:bool)
   > Set SlidingRBEnable(SlidingRBEnable:bool=AND Boolean(A=@n21, B=Get bRigidBodyEnabled())) -> (Output_Get:bool)
   n22 = Event BlueprintProcessAnimInputEvent()
   > Set OnZipline(OnZipline:bool=true) -> (Output_Get:bool)
   > n23 = Branch(Condition=IsNamed(self=@n22.AnimInputEvent, InEventName=aie.Player.Zipline.Entered))
   > Set OnZipline(OnZipline=false)
   Branch(execute=@n23.else, Condition=IsNamed(self=@n22.AnimInputEvent, InEventName=aie.Player.Zipline.Exited))
   > Set OnZipline(OnZipline=true)
   Event BlueprintInitializeAnimation()
   > Set Skel Mesh(Skel Mesh=GetOwnerSkeletalMeshComponent())
   Event OnCharacterCustomizationLoaded()
   > ResetDynamics(InTeleportType=TeleportPhysics)
   > ResetDynamics(InTeleportType=ResetPhysics)
   Branch(Condition=Get bIsInFrontEnd())
  ]}
]
variables [
 BalletSpin_FE {type=boolean}
 floatingVar {type=number}
 Gravity Override Params {type=GravityOverrideParamsStruct, default {onGround_Moving {gravityOverride=<0 0 -1000>}, onGround_Standing {gravityOverride=<0 0 -1200>}}}
 GravityOverride {type=Vector}
 head rot {type=Rotator}
 isRunFloat {type=number}
 IsRunning {type=boolean}
 isSliding {type=number}
 MeeleWeapon {type=number}
 Moving2d {type=number}
 OnZipline {type=boolean}
 Skel Mesh {type=SkeletalMeshComponent}
 SkirtRotation {type=number}
 SlidingRBdisabled {type=boolean}
 SlidingRBEnable {type=boolean}
 X percentage {type=number, default=0.7}
 Y percentage {type=number}
 Z percentage {type=number, default=0.8}
]
